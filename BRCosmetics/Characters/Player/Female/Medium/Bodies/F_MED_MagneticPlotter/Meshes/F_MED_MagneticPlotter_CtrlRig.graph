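# Control Rig for the F_MED_MagneticPlotter female body, driving secondary deformation of clothing and accessory bones.
# After an optional debug control-to-leg-bone pass, the shared DeformRig function runs, then knee pad, surface constraint, sleeve and buckle, and jacket/tissue passes.
# Weighted parent constraints rivet accessories to the body, while deform-rig anim attributes blend corrective offsets.
# Jacket, pocket and tissue follow sign-filtered thigh swing-twist using aim constraints and remapped weights.
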ControlRigBlueprint /BRCosmetics/Characters/Player/Female/Medium/Bodies/F_MED_MagneticPlotter/Meshes/F_MED_MagneticPlotter_CtrlRig
graphs [
 New Function_ContainedGraph {nodes [
   n1 = RigVMFunctionEntryNode()
   RigVMFunction_ControlFlowBranch = &RigVMFunction_ControlFlowBranch(Condition=@n1.Debug, ExecuteContext=@n1.ExecuteContext)
   RigVMFunctionReturnNode(ExecuteContext=@RigVMFunction_ControlFlowBranch.Completed)
   &RigUnit_SetTransform(
    Item=(Type=Bone,Name="thigh_l")
    Weight=1
    bInitial=false
    bPropagateToChildren=true
    Value=&RigUnit_GetTransform(Item=(Type=Control,Name="thigh_l_ctrl"), bInitial=false).Transform
    ExecutePin=@RigVMFunction_ControlFlowBranch.True
   )
   > &RigUnit_SetTransform(
    Item=(Type=Bone,Name="thigh_r")
    Weight=1
    bInitial=False
    Value=&RigUnit_GetTransform(Item=(Type=Control,Name="thigh_r_ctrl"), bInitial=False).Transform
   )
   > &RigUnit_SetTransform(
    Item=(Type=Bone,Name="calf_l")
    Weight=1
    bInitial=False
    Value=&RigUnit_GetTransform(Item=(Type=Control,Name="calf_l_ctrl"), bInitial=False).Transform
   )
   > &RigUnit_SetTransform(
    Item=(Type=Bone,Name="calf_r")
    Weight=1
    bInitial=False
    Value=&RigUnit_GetTransform(Item=(Type=Control,Name="calf_r_ctrl"), bInitial=False).Transform
   )
  ]}
 New Function_ContainedGraph {nodes [
   RigVMFunctionEntryNode()
   > &RigUnit_ParentConstraint(
    AdvancedSettings=(RotationOrderForFilter=XZY)
    Child=(Type=Bone,Name="base_buckle")
    Parents="((Item=(Type=Bone,Name=\"spine_03\"),Weight=0.0022),(Item=(Type=Bone,Name=\"thigh_twist_01_r\"),Weight=0.0024),(Item=(Type=Bone,Name=\"deform_groin_r\"),Weight=0.0029),(Item=(Type=Bone,Name=\"thigh_twist_01_l\"),Weight=0.0029),(Item=(Type=Bone,Name=\"deform_groin_l ...(truncated)"
    Weight=1
   )
   > &RigUnit_ParentConstraint(
    AdvancedSettings=(RotationOrderForFilter=XZY)
    Child=(Type=Bone,Name="base_radio")
    Parents=((Item=(Type=Bone,Name="spine_02"),Weight=0.0017),(Item=(Type=Bone,Name="deform_glute_l"),Weight=0.0666),(Item=(Type=Bone,Name="pelvis"),Weight=0.9317))
    Weight=1
   )
   > &RigUnit_ParentConstraint(
    AdvancedSettings=(RotationOrderForFilter=XZY)
    Child=(Type=Bone,Name="base_master_zipper_l")
    Parents=((Item=(Type=Bone,Name="calf_twist_02_l"),Weight=0.4685),(Item=(Type=Bone,Name="calf_l"),Weight=0.5315))
    Weight=1
   )
   > &RigUnit_ParentConstraint(
    AdvancedSettings=(RotationOrderForFilter=XZY)
    Child=(Type=Bone,Name="base_master_zipper_r")
    Parents=((Item=(Type=Bone,Name="deform_knee_out_r"),Weight=0.0005),(Item=(Type=Bone,Name="calf_twist_02_r"),Weight=0.4717),(Item=(Type=Bone,Name="calf_r"),Weight=0.5279))
    Weight=1
   )
   > &RigUnit_ParentConstraint(
    AdvancedSettings=(RotationOrderForFilter=XZY)
    Child=(Type=Bone,Name="base_tissue_0")
    Parents=((Item=(Type=Bone,Name="thigh_twist_01_r"),Weight=0.022),(Item=(Type=Bone,Name="deform_glute_r"),Weight=0.1789),(Item=(Type=Bone,Name="pelvis"),Weight=0.7991))
    Weight=1
   )
   > &RigUnit_ParentConstraint(
    AdvancedSettings=(RotationOrderForFilter=XZY)
    Child=(Type=Bone,Name="base_front_sleeve_l_0")
    Parents="((Item=(Type=Bone,Name=\"spine_04\"),Weight=0.0001),(Item=(Type=Bone,Name=\"deform_groin_r\"),Weight=0.0033),(Item=(Type=Bone,Name=\"thigh_twist_01_r\"),Weight=0.0078),(Item=(Type=Bone,Name=\"spine_02\"),Weight=0.0354),(Item=(Type=Bone,Name=\"thigh_twist_01_l\"),Wei ...(truncated)"
    Weight=1
   )
   > &RigUnit_ParentConstraint(
    AdvancedSettings=(RotationOrderForFilter=XZY)
    Child=(Type=Bone,Name="base_front_sleeve_r_0")
    Parents="((Item=(Type=Bone,Name=\"spine_04\"),Weight=0.0013),(Item=(Type=Bone,Name=\"deform_groin_l\"),Weight=0.0117),(Item=(Type=Bone,Name=\"thigh_twist_01_l\"),Weight=0.0145),(Item=(Type=Bone,Name=\"thigh_twist_01_r\"),Weight=0.0207),(Item=(Type=Bone,Name=\"deform_groin_r ...(truncated)"
    Weight=1
   )
   > &RigUnit_ParentConstraint(
    AdvancedSettings=(RotationOrderForFilter=XZY)
    Child=(Type=Bone,Name="rivet_for_aim_sleeve_l")
    Parents=((Item=(Type=Bone,Name="pelvis"),Weight=0.0057),(Item=(Type=Bone,Name="deform_knee_out_l"),Weight=0.0075),(Item=(Type=Bone,Name="thigh_twist_01_l"),Weight=0.1637),(Item=(Type=Bone,Name="thigh_l"),Weight=0.8231))
    Weight=1
   )
   > &RigUnit_ParentConstraint(
    AdvancedSettings=(RotationOrderForFilter=XZY)
    Child=(Type=Bone,Name="rivet_for_aim_sleeve_r")
    Parents="((Item=(Type=Bone,Name=\"deform_knee_out_r\"),Weight=0.0016),(Item=(Type=Bone,Name=\"deform_groin_r\"),Weight=0.0057),(Item=(Type=Bone,Name=\"pelvis\"),Weight=0.0165),(Item=(Type=Bone,Name=\"thigh_twist_01_r\"),Weight=0.2411),(Item=(Type=Bone,Name=\"thigh_r\"),Weigh ...(truncated)"
    Weight=1
   )
   > RigVMFunctionReturnNode()
  ]}
 New Function_ContainedGraph {nodes [
   RigVMFunctionEntryNode()
   > &RigUnit_ModifyTransforms(
    ItemToModify="((Item=(Type=Bone,Name=\"kneepad_calf_position_l\"),Transform=(Rotation=(X=0,Y=0,Z=0.5,W=0.866),Translation=(X=0,Y=0,Z=0),Scale3D=(X=1,Y=1,Z=1))),(Item=(Type=Bone,Name=\"dyn_kneepad_l\"),Transform=(Rotation=(X=0,Y=0.0611,Z=0,W=0.9981),Translation=(X=-0.6518,Y= ...(truncated)"
    WeightMinimum=0
    Weight=RigVMDispatchNode(BoneName=root, Default=0, Name=deformRig_l_knee_alpha) {template="DISPATCH_RigDispatch_GetAnimAttribute(in Name,in BoneName,in Default,out Value,out Found)"}.Value
   )
   > &RigUnit_ModifyTransforms(
    ItemToModify="((Item=(Type=Bone,Name=\"kneepad_calf_position_r\"),Transform=(Rotation=(X=0,Y=0,Z=0.5,W=0.866),Translation=(X=0,Y=0,Z=0),Scale3D=(X=1,Y=1,Z=1))),(Item=(Type=Bone,Name=\"dyn_kneepad_r\"),Transform=(Rotation=(X=0.0009,Y=0.0611,Z=0.0149,W=0.998),Translation=(X=- ...(truncated)"
    WeightMinimum=0
    Weight=RigVMDispatchNode(BoneName=root, Default=0, Name=deformRig_r_knee_alpha) {template="DISPATCH_RigDispatch_GetAnimAttribute(in Name,in BoneName,in Default,out Value,out Found)"}.Value
   )
   > RigVMFunctionReturnNode()
  ]}
 New Function_ContainedGraph {nodes [
   RigVMFunctionEntryNode()
   > &RigUnit_AimConstraintLocalSpaceOffset(
    Child=(Type=Bone,Name="base_front_sleeve_l_0")
    Filter=(bX=True,bY=True,bZ=True)
    Parents=((Item=(Type=Bone,Name="rivet_for_aim_sleeve_l"),Weight=1))
    Weight=1
    WorldUp=(Target=(X=0,Y=0,Z=1),Kind=Location,Space=(Type=Bone,Name="rivet_for_aim_sleeve_l"))
   )
   > &RigUnit_AimConstraintLocalSpaceOffset(
    Child=(Type=Bone,Name="base_front_sleeve_r_0")
    Filter=(bX=True,bY=True,bZ=True)
    Parents=((Item=(Type=Bone,Name="rivet_for_aim_sleeve_r"),Weight=1))
    Weight=1
    WorldUp=(Target=(X=0,Y=0,Z=1),Kind=Location,Space=(Type=Bone,Name="rivet_for_aim_sleeve_r"))
   )
   n1 = RigVMDispatchNode(BoneName=root, Default=0, Name=deformRig_r_thigh_up_alpha) {template="DISPATCH_RigDispatch_GetAnimAttribute(in Name,in BoneName,in Default,out Value,out Found)"}
   n2 = RigVMDispatchNode(BoneName=root, Default=0, Name=deformRig_l_thigh_up_alpha) {template="DISPATCH_RigDispatch_GetAnimAttribute(in Name,in BoneName,in Default,out Value,out Found)"}
   DISPATCH_RigVMDispatch_If = RigVMDispatchNode(
    Condition=/Script/RigVM.RigVMFunction_MathDoubleGreaterEqual(B=@n1.Value, A=@n2.Value) {is=struct RigVMUnitNode, template="GreaterEqual::Execute(in A,in B,out Result)"}.Result
    False=@n1.Value
    True=@n2.Value
   ) {template="DISPATCH_RigVMDispatch_If(in Condition,in True,in False,out Result)"}
   AlphaInterp = &RigVMFunction_AlphaInterp(
    InterpSpeedDecreasing=20
    InterpSpeedIncreasing=10
    Scale=1
    bClampResult=false
    bMapRange=false
    Value=&RigVMFunction_MathDoubleRemap(SourceMaximum=1, SourceMinimum=0, TargetMaximum=1, TargetMinimum=0, Value=@DISPATCH_RigVMDispatch_If.Result).Result
   )
   > &RigUnit_ModifyTransforms(
    ItemToModify=((Item=(Type=Bone,Name="base_buckle"),Transform=(Rotation=(X=0,Y=-0.061,Z=0,W=0.9981),Translation=(X=-5.2424,Y=0,Z=-0.9472),Scale3D=(X=0.9,Y=0.9,Z=0.9))))
    WeightMinimum=0
    Weight=@AlphaInterp.Result
   )
   Remap = &RigVMFunction_MathDoubleRemap(SourceMaximum=1, SourceMinimum=0, TargetMaximum=1, TargetMinimum=0, Value=@n2.Value)
   > &RigUnit_ModifyTransforms(
    ItemToModify=((Item=(Type=Bone,Name="base_front_sleeve_l_0"),Transform=(Rotation=(X=-0.1901,Y=0.0736,Z=-0.0775,W=0.9759),Translation=(X=-0.9023,Y=0.246,Z=-0.0242),Scale3D=(X=1,Y=1,Z=1))))
    WeightMinimum=0
    Weight=@Remap.Result
   )
   > &RigUnit_ModifyTransforms(
    ItemToModify=((Item=(Type=Bone,Name="base_front_sleeve_r_0"),Transform=(Rotation=(X=0.1989,Y=0.1095,Z=0.0573,W=0.9722),Translation=(X=-0.0662,Y=0.8066,Z=0.4822),Scale3D=(X=1,Y=1,Z=1))))
    WeightMinimum=0
    Weight=&RigVMFunction_MathDoubleRemap(SourceMaximum=1, SourceMinimum=0, TargetMaximum=1, TargetMinimum=0, Value=@n1.Value).Result
   )
   > RigVMFunctionReturnNode()
  ]}
 New Function_ContainedGraph {nodes [
   RigVMFunctionEntryNode()
   > RigVMFunction_Sequence = RigVMAggregateNode()
   &RigUnit_ParentConstraint(
    AdvancedSettings=(InterpolationType=Average,RotationOrderForFilter=XZY)
    Child=(Type=Bone,Name="thigh_l_for_cons")
    Parents=((Item=(Type=Bone,Name="thigh_l"),Weight=1))
    Weight=1
    ExecutePin=@RigVMFunction_Sequence.A
   )
   RigUnit_GetTransform = &RigUnit_GetTransform(Item=(Type=Bone,Name="thigh_l_for_cons"), Space=LocalSpace, bInitial=false)
   n1 = &RigVMFunction_MathQuaternionSwingTwist(TwistAxis=(X=1,Y=0,Z=0), Input=@RigUnit_GetTransform.Transform.Rotation)
   RigVMFunction_MathQuaternionToEuler = &RigVMFunction_MathQuaternionToEuler(RotationOrder=ZXY, Value=@n1.Swing)
   LessEqual = /Script/RigVM.RigVMFunction_MathDoubleLessEqual(B=0, A=@RigVMFunction_MathQuaternionToEuler.Result.Y) {is=struct RigVMUnitNode, template="LessEqual::Execute(in A,in B,out Result)"}
   GreaterEqual = &RigVMFunction_MathDoubleGreaterEqual(B=0, A=@RigVMFunction_MathQuaternionToEuler.Result.Z) {template="GreaterEqual::Execute(in A,in B,out Result)"}
   RigVMFunction_MathQuaternionFromEuler = &RigVMFunction_MathQuaternionFromEuler(
    Euler=(X=0,Y=0,Z=0)
    RotationOrder=ZXY
    Euler.Y=RigVMDispatchNode(False=0, Condition=@LessEqual.Result, True=@RigVMFunction_MathQuaternionToEuler.Result.Y) {template="DISPATCH_RigVMDispatch_If(in Condition,in True,in False,out Result)"}.Result
    Euler.Z=RigVMDispatchNode(False=0, Condition=@GreaterEqual.Result, True=@RigVMFunction_MathQuaternionToEuler.Result.Z) {template="DISPATCH_RigVMDispatch_If(in Condition,in True,in False,out Result)"}.Result
    Euler.X=&RigVMFunction_MathQuaternionToEuler(RotationOrder=ZXY, Value=@n1.Twist).Result.X
   )
   > &RigUnit_ModifyTransforms(
    ItemToModify=((Item=(Type=Bone,Name="thigh_l_for_cons_set_rotation"),Transform=(Rotation=(X=0,Y=0,Z=0,W=1),Translation=(X=0,Y=0,Z=0),Scale3D=(X=1,Y=1,Z=1))))
    Weight=1
    WeightMinimum=0
    ItemToModify.0.Transform.Rotation=@RigVMFunction_MathQuaternionFromEuler.Result
   )
   > &RigUnit_ParentConstraint(
    AdvancedSettings=(InterpolationType=Average,RotationOrderForFilter=XZY)
    Child=(Type=Bone,Name="base_jacket_l_0")
    Parents=((Item=(Type=Bone,Name="thigh_l_for_cons_set_rotation"),Weight=1))
    Weight=1
   )
   > &RigUnit_ParentConstraint(
    AdvancedSettings=(InterpolationType=Average,RotationOrderForFilter=XZY)
    Child=(Type=Bone,Name="base_jacket_l_1")
    Parents=((Item=(Type=Bone,Name="thigh_l_for_cons_set_rotation"),Weight=1))
    Weight=1
   )
   > RigVMFunctionReturnNode()
   &RigUnit_ParentConstraint(
    AdvancedSettings=(InterpolationType=Average,RotationOrderForFilter=XZY)
    Child=(Type=Bone,Name="thigh_r_for_cons")
    Parents=((Item=(Type=Bone,Name="thigh_r"),Weight=1))
    Weight=1
    ExecutePin=@RigVMFunction_Sequence.B
   )
   n2 = &RigVMFunction_MathQuaternionSwingTwist(
    TwistAxis=(X=1,Y=0,Z=0)
    Input=&RigUnit_GetTransform(Item=(Type=Bone,Name="thigh_r_for_cons"), Space=LocalSpace, bInitial=false).Transform.Rotation
   )
   n3 = &RigVMFunction_MathQuaternionToEuler(RotationOrder=ZXY, Value=@n2.Swing)
   n4 = RigVMDispatchNode(
    False=0
    Condition=&RigVMFunction_MathDoubleLessEqual(B=0, A=@n3.Result.Y) {template="LessEqual::Execute(in A,in B,out Result)"}.Result
    True=@n3.Result.Y
   ) {template="DISPATCH_RigVMDispatch_If(in Condition,in True,in False,out Result)"}
   n5 = &RigVMFunction_MathDoubleGreaterEqual(B=0, A=@n3.Result.Z) {template="GreaterEqual::Execute(in A,in B,out Result)"}
   > &RigUnit_ModifyTransforms(
    ItemToModify=((Item=(Type=Bone,Name="thigh_r_for_cons_set_rotation"),Transform=(Rotation=(X=0,Y=0,Z=0,W=1),Translation=(X=0,Y=0,Z=0),Scale3D=(X=1,Y=1,Z=1))))
    Weight=1
    WeightMinimum=0
    ItemToModify.0.Transform.Rotation=&RigVMFunction_MathQuaternionFromEuler(
     Euler=(X=0,Y=0,Z=0)
     RotationOrder=ZXY
     Euler.Y=@n4.Result
     Euler.Z=RigVMDispatchNode(False=0, Condition=@n5.Result, True=@n3.Result.Z) {template="DISPATCH_RigVMDispatch_If(in Condition,in True,in False,out Result)"}.Result
     Euler.X=&RigVMFunction_MathQuaternionToEuler(RotationOrder=ZXY, Value=@n2.Twist).Result.X
    ).Result
   )
   > &RigUnit_ParentConstraint(
    AdvancedSettings=(InterpolationType=Average,RotationOrderForFilter=XZY)
    Child=(Type=Bone,Name="base_jacket_r_0")
    Parents=((Item=(Type=Bone,Name="thigh_r_for_cons_set_rotation"),Weight=1))
    Weight=1
   )
   > &RigUnit_ParentConstraint(
    AdvancedSettings=(InterpolationType=Average,RotationOrderForFilter=XZY)
    Child=(Type=Bone,Name="base_jacket_r_1")
    Parents=((Item=(Type=Bone,Name="thigh_r_for_cons_set_rotation"),Weight=1))
    Weight=1
   )
   DISPATCH_RigVMDispatch_If = RigVMDispatchNode(
    Condition=&RigVMFunction_MathDoubleGreaterEqual(A=@RigVMFunction_MathQuaternionToEuler.Result.Z, B=@n3.Result.Z) {template="GreaterEqual::Execute(in A,in B,out Result)"}.Result
    True=@RigVMFunction_MathQuaternionToEuler.Result.Z
    False=@n3.Result.Z
   ) {template="DISPATCH_RigVMDispatch_If(in Condition,in True,in False,out Result)"}
   n6 = &RigVMFunction_MathQuaternionFromEuler(Euler=(X=0,Y=0,Z=0), RotationOrder=ZXY, Euler.Z=@DISPATCH_RigVMDispatch_If.Result)
   > &RigUnit_ModifyTransforms(
    ItemToModify=((Item=(Type=Bone,Name="base_jacket_c"),Transform=(Rotation=(X=0,Y=-0.7071,Z=0,W=0.7071),Translation=(X=-4.5,Y=0,Z=-4.5),Scale3D=(X=1,Y=1,Z=1))))
    WeightMinimum=0
    Weight=&RigVMFunction_MathDoubleRemap(
     SourceMaximum=60
     SourceMinimum=0
     TargetMaximum=1
     TargetMinimum=0
     Value=&RigVMFunction_MathQuaternionToEuler(RotationOrder=ZYX, Value=@n6.Result).Result.Z
    ).Result
   )
   &RigUnit_ParentConstraint(
    AdvancedSettings=(RotationOrderForFilter=XZY)
    Child=(Type=Bone,Name="rivet_for_pocket")
    Parents="((Item=(Type=Bone,Name=\"dyn_jacket_c_end\"),Weight=0.0027),(Item=(Type=Bone,Name=\"thigh_twist_01_r\"),Weight=0.0245),(Item=(Type=Bone,Name=\"dyn_jacket_r_1\"),Weight=0.053),(Item=(Type=Bone,Name=\"pelvis\"),Weight=0.0592),(Item=(Type=Bone,Name=\"thigh_r\"),Weight= ...(truncated)"
    Weight=1
    ExecutePin=@RigVMFunction_Sequence.C
   )
   > &RigUnit_AimConstraintLocalSpaceOffset(
    Child=(Type=Bone,Name="pocket_for_aim_cons")
    Filter=(bX=True,bY=True,bZ=True)
    Parents=((Item=(Type=Bone,Name="rivet_for_pocket"),Weight=1))
    UpAxis=(X=0,Y=1,Z=0)
    Weight=1
    WorldUp=(Target=(X=0,Y=1,Z=0),Kind=Location,Space=(Type=Bone,Name="rivet_for_pocket"))
   )
   Remap = &RigVMFunction_MathDoubleRemap(SourceMaximum=-35, SourceMinimum=0, TargetMaximum=1, TargetMinimum=0, Value=@n4.Result)
   AlphaInterp = &RigVMFunction_AlphaInterp(InterpSpeedDecreasing=5, InterpSpeedIncreasing=5, Scale=1, bClampResult=true, bMapRange=true, Value=@Remap.Result)
   > &RigUnit_ModifyTransforms(
    ItemToModify=((Item=(Type=Bone,Name="dyn_pocket"),Transform=(Rotation=(X=0,Y=0,Z=0,W=1),Translation=(X=0,Y=0,Z=0),Scale3D=(X=1,Y=1,Z=1))))
    WeightMinimum=0
    Weight=@AlphaInterp.Result
    ItemToModify.0.Transform.Rotation=&RigUnit_GetTransform(Item=(Type=Bone,Name="pocket_for_aim_cons"), Space=LocalSpace, bInitial=False).Transform.Rotation
   )
   > &RigUnit_ParentConstraint(
    AdvancedSettings=(RotationOrderForFilter=XZY)
    Child=(Type=Bone,Name="rivet_for_aim_tissue")
    Parents="((Item=(Type=Bone,Name=\"dyn_jacket_l_1_end\"),Weight=0.0036),(Item=(Type=Bone,Name=\"dyn_jacket_r_0\"),Weight=0.0293),(Item=(Type=Bone,Name=\"pelvis\"),Weight=0.0305),(Item=(Type=Bone,Name=\"dyn_jacket_c\"),Weight=0.0488),(Item=(Type=Bone,Name=\"dyn_jacket_r_1\"),W ...(truncated)"
    Weight=1
   )
   > &RigUnit_AimConstraintLocalSpaceOffset(
    Child=(Type=Bone,Name="base_tissue_3")
    Filter=(bX=True,bY=True,bZ=True)
    Parents=((Item=(Type=Bone,Name="rivet_for_aim_tissue"),Weight=1))
    Weight=1
    WorldUp=(Target=(X=0,Y=0,Z=0),Kind=Direction,Space=(Type=None,Name="None"))
   )
   > &RigUnit_RotationConstraintLocalSpaceOffset(
    AdvancedSettings=(InterpolationType=Average,RotationOrderForFilter=XZY)
    Child=(Type=Bone,Name="base_tissue_3")
    Filter=(bX=True,bY=True,bZ=True)
    Parents=((Item=(Type=Bone,Name="None"),Weight=1))
    Weight=1
   )
   > &RigUnit_ModifyTransforms(
    ItemToModify=((Item=(Type=Bone,Name="base_tissue_0"),Transform=(Rotation=(X=0,Y=0,Z=0,W=1),Translation=(X=0,Y=0,Z=0),Scale3D=(X=1,Y=1,Z=1))))
    WeightMinimum=0
    ItemToModify.0.Transform.Rotation=&RigUnit_GetTransform(Item=(Type=Bone,Name="base_tissue_3"), Space=LocalSpace, bInitial=False).Transform.Rotation
    Weight=&RigVMFunction_MathDoubleRemap(SourceMaximum=-90, SourceMinimum=0, TargetMaximum=0, TargetMinimum=1, Value=@n4.Result).Result
   )
   > &RigUnit_ModifyTransforms(
    ItemToModify=((Item=(Type=Bone,Name="base_tissue_0"),Transform=(Rotation=(X=-0.1482,Y=-0.0913,Z=0.1162,W=0.9779),Translation=(X=0,Y=0,Z=0),Scale3D=(X=1,Y=1,Z=1))))
    WeightMinimum=0
    Weight=&RigVMFunction_MathDoubleRemap(SourceMaximum=-90, SourceMinimum=0, TargetMaximum=1, TargetMinimum=0, Value=@n4.Result).Result
   )
  ]}
 RigVMFunctionLibrary {nodes [
   controls = RigVMCollapseNode(Debug=False, ExecuteContext=())
   jacket_tissue = RigVMCollapseNode(ExecuteContext=())
   knee pads = RigVMCollapseNode(ExecuteContext=())
   sleeves and buckle = RigVMCollapseNode(ExecuteContext=())
   surface constraints before jacket = RigVMCollapseNode(ExecuteContext=())
  ]}
 RigVMFunction_Sequence_1_ContainedGraph {nodes [
   n1 = RigVMFunctionEntryNode()
   RigVMFunction_Sequence = &RigVMFunction_Sequence(ExecuteContext=@n1.ExecuteContext)
   n2 = &RigVMFunction_Sequence(ExecuteContext=@RigVMFunction_Sequence.B)
   RigVMFunctionReturnNode(A=@RigVMFunction_Sequence.A, B=@n2.A, C=@n2.B)
  ]}
 RigVMModel {nodes [
   RigUnit_BeginExecution = &RigUnit_BeginExecution()
   controls = RigVMFunctionReferenceNode(Debug=false, ExecuteContext=@RigUnit_BeginExecution.ExecutePin)
   > FN_DeformRig_FNC = RigVMFunctionReferenceNode(
    DeformRig_ Curve="(deformRig_l_upperarm_up_alpha_72_#1=False,deformRig_r_upperarm_up_alpha_73_#2=False,deformRig_l_upperarm_fr_alpha_74_#3=False,deformRig_r_upperarm_fr_alpha_75_CB51B8 ...(truncated)"
    DeformRig_Bool="(upperarm_1_#4=True,elbow_3_#5=True,wrist_7_#6=True,groin_8_#7=True,glute_17_#8=True,knee_9_40C408324CA261 ...(truncated)"
    DeformRig_Debug="(upperarm_debug=False,wrist_debug=False,lat_debug=False,groinGlute_debug=False,upperarm_debug_18_#4=False,wrist_debug_20_#6=False,groinGlute_debug_27_#7=False,lat_deb ...(truncated)"
    DeformRig_Value="(upperarm_up_18_#9=(Translation=(X=0,Y=0,Z=8),Scale3D=(X=0.7,Y=1,Z=1)),upperarm_fwd_fr_19_#10=(Translation=(X=0,Y=10,Z=0),Scale3D=(X=0.5,Y=1,Z=1)),upperarm_fwd_bk_20_9973AAD84EE0A1E349BC9E85067D4A3 ...(truncated)"
   )
   > knee pads = RigVMFunctionReferenceNode()
   > surface constraints before jacket = RigVMFunctionReferenceNode()
   > sleeves and buckle = RigVMFunctionReferenceNode()
   > jacket_tissue = RigVMFunctionReferenceNode()
  ]}
]
hierarchy [
 root {loc=<0 0 0>, rot=<0 0 0 1>}
 pelvis {parent=root, loc=<0 -0.5301 98.0571>, rot=<0.0127 0.707 -0.0127 -0.707>}
 spine_01 {parent=pelvis, loc=<6.1208 -0.9328 0>, rot=<0 0 0.034 -0.9994>}
 spine_02 {parent=spine_01, loc=<6.5997 -0.2095 0>, rot=<0 0 0.0238 -0.9997>}
 spine_03 {parent=spine_02, loc=<6.4914 -0.1237 0>, rot=<0 0 0.0149 -0.9999>}
 spine_04 {parent=spine_03, loc=<6.9598 0.0794 0>, rot=<0 0 -0.0132 -0.9999>}
 spine_05 {parent=spine_04, loc=<6.9987 0 0>, rot=<0 0 -0.0015 -1>}
 attach_cape {parent=spine_05, loc=<9.3216 -8.4501 0>, rot=<0.0616 -0.7044 -0.0616 -0.7044>}
 neck_01 {parent=spine_05, loc=<11.3138 0.1529 0>, rot=<0 0 -0.2003 -0.9797>}
 neck_02 {parent=neck_01, loc=<4.5627 0 0>, rot=<0 0 0.0504 -0.9987>}
 head {parent=neck_02, loc=<5.3914 0 0>, rot=<0 0 0.115 -0.9934>}
 base_main_front_hair {parent=head, loc=<8.4691 10.1649 0.1758>, rot=<0.518 0.4813 0.522 0.477>}
 dyn_main_front_hair {parent=base_main_front_hair, loc=<0 0 0>, rot=<0 0 0 1>}
 dyn_front_hair_0 {parent=dyn_main_front_hair, loc=<0.0405 -1.1288 -1.2071>, rot=<-0.0791 -0.4201 -0.0944 -0.8991>}
 dyn_front_hair_1 {parent=dyn_main_front_hair, loc=<2.3764 1.1468 -1.3802>, rot=<-0.1703 -0.4175 -0.1845 -0.8733>}
 dyn_front_hair_2 {parent=dyn_main_front_hair, loc=<2.7765 2.6626 0.5954>, rot=<0.0003 -0.3962 -0.4111 -0.821>}
 base_main_back_hair {parent=head, loc=<1.8426 -9.6915 0.0198>, rot=<0.408 -0.5775 0.5775 -0.408>}
 dyn_main_back_hair {parent=base_main_back_hair, loc=<0 0 0>, rot=<0 0 0 1>}
 base_main_back_hair_new {parent=head, loc=<1.8426 -9.6915 0.0198>, rot=<0.0806 -0.7025 0.7025 -0.0806>}
 dyn_main_back_hair_new {parent=base_main_back_hair_new, loc=<0 0 0>, rot=<0 0 0 1>}
 cap1 {parent=head, loc=<5.9913 -9.1509 0>, rot=<0.5816 -0.4021 0.4021 -0.5816>}
 cap {parent=cap1, loc=<0 0 0>, rot=<0 0 0 1>}
 rivet_for_aim_cap {parent=head, loc=<13.7863 -29.8033 0>, rot=<0.5816 -0.4021 0.4021 -0.5816>}
 base_cap {parent=head, loc=<5.9913 -9.1509 0>, rot=<0.5816 -0.4021 0.4021 -0.5816>}
 dyn_cap {parent=base_cap, loc=<0 0 0>, rot=<0 0 0 1>}
 attach_backpack {parent=spine_05, loc=<0 -8.35 0>, rot=<-0.0093 -0.707 0.0093 -0.707>}
 clavicle_l {parent=spine_05, loc=<5.4831 -0.1053 -3.2946>, rot=<0.0465 -0.7289 -0.009 -0.6829>}
 upperarm_l {parent=clavicle_l, loc=<9.8658 -1.2136 -0.4105>, rot=<0.0164 -0.3245 -0.0269 -0.9454>}
 lowerarm_l {parent=upperarm_l, loc=<26.465 -0.8345 0.259>, rot=<0.0036 0.0161 -0.2298 -0.9731>}
 hand_l {parent=lowerarm_l, loc=<22.7355 -0.3203 -0.6105>, rot=<-0.6784 -0.0223 0.0735 -0.7307>}
 hand_attach_l {parent=hand_l, loc=<8.5 -3 0>, rot=<0.7071 0.7071 0 0>}
 weapon_l {parent=hand_l, loc=<-0.1974 -3.4111 0.3811>, rot=<0 0 0 1>}
 thumb_01_l {parent=hand_l, loc=<1.3118 -0.3191 -2.459>, rot=<0.6546 -0.4042 0.0664 -0.6354>}
 thumb_02_l {parent=thumb_01_l, loc=<5.9336 -0.0448 0.0401>, rot=<-0.0037 0.0213 0.298 -0.9543>}
 thumb_03_l {parent=thumb_02_l, loc=<2.2223 0.3682 -0.0907>, rot=<0.0118 -0.013 0.0066 -0.9998>}
 index_metacarpal_l {parent=hand_l, loc=<1.6535 0.0183 -2.96>, rot=<0.0158 -0.0124 -0.0173 -0.9996>}
 index_01_l {parent=index_metacarpal_l, loc=<8.3008 0.0357 -0.0524>, rot=<-0.0044 0.0485 0.0924 -0.9945>}
 index_02_l {parent=index_01_l, loc=<3.3587 -0.3004 0.0452>, rot=<-0.002 0.0027 0.3225 -0.9466>}
 index_03_l {parent=index_02_l, loc=<2.5578 -0.123 -0.0252>, rot=<-0.002 0.0067 0.2558 -0.9667>}
 middle_metacarpal_l {parent=hand_l, loc=<2.1856 -0.1725 -0.5984>, rot=<0.006 0.0168 -0.0329 -0.9993>}
 middle_01_l {parent=middle_metacarpal_l, loc=<7.3275 -0.2447 0.0129>, rot=<-0.0208 0.0042 0.0858 -0.9961>}
 middle_02_l {parent=middle_01_l, loc=<4.0123 0.1401 0.1175>, rot=<-0.0218 0.0258 0.3512 -0.9357>}
 middle_03_l {parent=middle_02_l, loc=<2.6077 0.0534 0.1373>, rot=<0.0082 0.0188 0.1708 -0.9851>}
 ring_metacarpal_l {parent=hand_l, loc=<2.2074 -0.3625 1.3638>, rot=<-0.0284 0.0722 -0.0458 -0.9959>}
 ring_01_l {parent=ring_metacarpal_l, loc=<7.0295 -0.6197 -0.1446>, rot=<-0.0489 -0.0252 0.1892 -0.9804>}
 ring_02_l {parent=ring_01_l, loc=<3.6174 0.4647 -0.1027>, rot=<0 0 0.2806 -0.9598>}
 ring_03_l {parent=ring_02_l, loc=<2.4152 0.114 0.0199>, rot=<0 0 0.1189 -0.9929>}
 pinky_metacarpal_l {parent=hand_l, loc=<1.7616 -0.9753 3.3814>, rot=<-0.0397 0.0814 -0.069 -0.9935>}
 pinky_01_l {parent=pinky_metacarpal_l, loc=<7.1709 -0.4831 -0.1094>, rot=<-0.0654 -0.0035 0.2242 -0.9723>}
 pinky_02_l {parent=pinky_01_l, loc=<2.8274 0.0893 -0.1043>, rot=<0.0037 0.0011 0.2854 -0.9584>}
 pinky_03_l {parent=pinky_02_l, loc=<2.2046 0.0307 -0.0078>, rot=<-0.0002 0.0098 0.1445 -0.9895>}
 deform_wrist_dn_base_l {parent=hand_l, loc=<0 -1.5326 0>, rot=<0 0 0 -1>}
 deform_wrist_dn_l {parent=deform_wrist_dn_base_l, loc=<0 0 0>, rot=<0 0 0 1>}
 deform_wrist_up_base_l {parent=hand_l, loc=<0 2.3009 0>, rot=<0 0 0 -1>}
 deform_wrist_up_l {parent=deform_wrist_up_base_l, loc=<0 0 0>, rot=<0 0 0 1>}
 lowerarm_twist_02_l {parent=lowerarm_l, loc=<10.6707 0 0>, rot=<0 0 0 1>}
 lowerarm_twist_01_l {parent=lowerarm_l, loc=<18.316 0 0>, rot=<0 0 0 1>}
 deform_elbow_in_base_l {parent=lowerarm_l, loc=<0.1834 3.6062 -0.2572>, rot=<0 0 0 -1>}
 deform_elbow_in_l {parent=deform_elbow_in_base_l, loc=<0 0 0>, rot=<0 0 0 1>}
 deform_elbow_out_base_l {parent=lowerarm_l, loc=<0 -4 0>, rot=<0 0 0 1>}
 deform_elbow_out_l {parent=deform_elbow_out_base_l, loc=<0 0 0>, rot=<0 0 0 1>}
 upperarm_twist_01_l {parent=upperarm_l, loc=<7.2135 -0.1134 0>, rot=<0 0 0 1>}
 deform_upperarm_bk_base_l {parent=upperarm_twist_01_l, loc=<-6.4053 -4.3746 0.314>, rot=<0 0 0 -1>}
 deform_upperarm_bk_l {parent=deform_upperarm_bk_base_l, loc=<0 0 0>, rot=<0 0 0 1>}
 deform_upperarm_fr_base_l {parent=upperarm_twist_01_l, loc=<-5.6312 5.092 0.3354>, rot=<0 0 0 -1>}
 deform_upperarm_fr_l {parent=deform_upperarm_fr_base_l, loc=<0 0 0>, rot=<0 0 0 1>}
 deform_upperarm_up_base_l {parent=upperarm_twist_01_l, loc=<-8.891 0.4825 3.9517>, rot=<0 0 0 -1>}
 deform_upperarm_up_l {parent=deform_upperarm_up_base_l, loc=<0 0 0>, rot=<0 0 0 1>}
 upperarm_twist_02_l {parent=upperarm_l, loc=<19.7998 -0.1134 0>, rot=<0 0 0 1>}
 clavicle_r {parent=spine_05, loc=<5.4831 -0.1053 3.2946>, rot=<0.7289 0.0465 -0.6829 0.009>}
 upperarm_r {parent=clavicle_r, loc=<-9.8658 1.2136 0.4105>, rot=<0.0164 -0.3245 -0.0269 -0.9454>}
 lowerarm_r {parent=upperarm_r, loc=<-26.465 0.8345 -0.259>, rot=<0.0036 0.0161 -0.2298 -0.9731>}
 hand_r {parent=lowerarm_r, loc=<-22.7355 0.3203 0.6106>, rot=<-0.6784 -0.0223 0.0735 -0.7307>}
 thumb_01_r {parent=hand_r, loc=<-1.3118 0.3191 2.459>, rot=<0.6546 -0.4042 0.0664 -0.6354>}
 thumb_02_r {parent=thumb_01_r, loc=<-5.9336 0.0448 -0.0401>, rot=<-0.0037 0.0213 0.298 -0.9543>}
 thumb_03_r {parent=thumb_02_r, loc=<-2.2223 -0.3682 0.0907>, rot=<0.0118 -0.013 0.0066 -0.9998>}
 index_metacarpal_r {parent=hand_r, loc=<-1.6535 -0.0183 2.96>, rot=<0.0158 -0.0124 -0.0173 -0.9996>}
 index_01_r {parent=index_metacarpal_r, loc=<-8.3008 -0.0357 0.0524>, rot=<-0.0044 0.0485 0.0924 -0.9945>}
 index_02_r {parent=index_01_r, loc=<-3.3587 0.3004 -0.0452>, rot=<-0.002 0.0027 0.3225 -0.9466>}
 index_03_r {parent=index_02_r, loc=<-2.5578 0.123 0.0252>, rot=<-0.002 0.0067 0.2558 -0.9667>}
 middle_metacarpal_r {parent=hand_r, loc=<-2.1856 0.1725 0.5984>, rot=<0.006 0.0168 -0.0329 -0.9993>}
 middle_01_r {parent=middle_metacarpal_r, loc=<-7.3275 0.2447 -0.0129>, rot=<-0.0208 0.0042 0.0858 -0.9961>}
 middle_02_r {parent=middle_01_r, loc=<-4.0123 -0.1401 -0.1175>, rot=<-0.0218 0.0258 0.3512 -0.9357>}
 middle_03_r {parent=middle_02_r, loc=<-2.6077 -0.0534 -0.1373>, rot=<0.0082 0.0188 0.1708 -0.9851>}
 ring_metacarpal_r {parent=hand_r, loc=<-2.2074 0.3625 -1.3638>, rot=<-0.0284 0.0722 -0.0458 -0.9959>}
 ring_01_r {parent=ring_metacarpal_r, loc=<-7.0295 0.6197 0.1446>, rot=<-0.0489 -0.0252 0.1892 -0.9804>}
 ring_02_r {parent=ring_01_r, loc=<-3.6174 -0.4647 0.1027>, rot=<0 0 0.2806 -0.9598>}
 ring_03_r {parent=ring_02_r, loc=<-2.4152 -0.114 -0.0199>, rot=<0 0 0.1189 -0.9929>}
 pinky_metacarpal_r {parent=hand_r, loc=<-1.7616 0.9753 -3.3814>, rot=<-0.0397 0.0814 -0.069 -0.9935>}
 pinky_01_r {parent=pinky_metacarpal_r, loc=<-7.1709 0.4831 0.1094>, rot=<-0.0654 -0.0035 0.2242 -0.9723>}
 pinky_02_r {parent=pinky_01_r, loc=<-2.8274 -0.0893 0.1043>, rot=<0.0037 0.0011 0.2854 -0.9584>}
 pinky_03_r {parent=pinky_02_r, loc=<-2.2046 -0.0307 0.0078>, rot=<-0.0002 0.0098 0.1445 -0.9895>}
 weapon_r {parent=hand_r, loc=<0.1974 3.4111 -0.3811>, rot=<0 0 0 1>}
 hand_attach_r {parent=hand_r, loc=<-8.5001 2.9999 0>, rot=<0 0 -0.7071 -0.7071>}
 deform_wrist_dn_base_r {parent=hand_r, loc=<-0.0001 1.5328 0>, rot=<0 0 0 -1>}
 deform_wrist_dn_r {parent=deform_wrist_dn_base_r, loc=<0 0 0>, rot=<0 0 0 1>}
 deform_wrist_up_base_r {parent=hand_r, loc=<-0.0002 -2.3007 -0.0001>, rot=<0 0 0 -1>}
 deform_wrist_up_r {parent=deform_wrist_up_base_r, loc=<0 0 0>, rot=<0 0 0 1>}
 lowerarm_twist_02_r {parent=lowerarm_r, loc=<-10.6707 0 0>, rot=<0 0 0 -1>}
 lowerarm_twist_01_r {parent=lowerarm_r, loc=<-18.316 0 0>, rot=<0 0 0 -1>}
 deform_elbow_in_base_r {parent=lowerarm_r, loc=<-0.1834 -3.6062 0.2572>, rot=<0 0 0 -1>}
 deform_elbow_in_r {parent=deform_elbow_in_base_r, loc=<0 0 0>, rot=<0 0 0 1>}
 deform_elbow_out_base_r {parent=lowerarm_r, loc=<0 4 -0.0001>, rot=<0 0 0 -1>}
 deform_elbow_out_r {parent=deform_elbow_out_base_r, loc=<0 0 0>, rot=<0 0 0 1>}
 base_main_chain_arm {parent=lowerarm_r, loc=<0 0 0>, rot=<0 -0.9914 0.1305 0>}
 dyn_main_chain_arm {parent=base_main_chain_arm, loc=<0 0 0>, rot=<0 0 0 1>}
 base_chain_arm {parent=dyn_main_chain_arm, loc=<15.708 -3.0514 -0.3353>, rot=<-0.4751 -0.5523 0.5034 -0.4646>}
 dyn_chain_arm {parent=base_chain_arm, loc=<0 0 0>, rot=<0 0 0 1>}
 upperarm_twist_01_r {parent=upperarm_r, loc=<-7.2135 0.1134 0>, rot=<0 0 0 -1>}
 deform_upperarm_bk_base_r {parent=upperarm_twist_01_r, loc=<6.4053 4.3746 -0.314>, rot=<0 0 0 -1>}
 deform_upperarm_bk_r {parent=deform_upperarm_bk_base_r, loc=<0 0 0>, rot=<0 0 0 1>}
 deform_upperarm_fr_base_r {parent=upperarm_twist_01_r, loc=<5.6312 -5.092 -0.3354>, rot=<0 0 0 -1>}
 deform_upperarm_fr_r {parent=deform_upperarm_fr_base_r, loc=<0 0 0>, rot=<0 0 0 1>}
 deform_upperarm_up_base_r {parent=upperarm_twist_01_r, loc=<8.891 -0.4825 -3.9517>, rot=<0 0 0 -1>}
 deform_upperarm_up_r {parent=deform_upperarm_up_base_r, loc=<0 0 0>, rot=<0 0 0 1>}
 upperarm_twist_02_r {parent=upperarm_r, loc=<-19.7998 0.1134 0>, rot=<0 0 0 -1>}
 rivet_for_aim_hair {parent=spine_05, loc=<9.8829 -22.9119 0.0198>, rot=<0.4285 -0.5625 0.5625 -0.4285>}
 deform_pec_aim_base_l {parent=spine_04, loc=<3.2456 9.3309 -1.2237>, rot=<0.0603 -0.4033 0.3318 -0.8507>}
 deform_pec_aim_l {parent=deform_pec_aim_base_l, loc=<0 0 0>, rot=<0 0 0 1>}
 deform_pec_base_l {parent=deform_pec_aim_l, loc=<9.3665 0 0>, rot=<0 0 0 1>}
 deform_pec_l {parent=deform_pec_base_l, loc=<0 0 0>, rot=<0 0 0 1>}
 deform_pec_aim_base_r {parent=spine_04, loc=<3.2452 9.3309 1.2237>, rot=<-0.4033 -0.0603 0.8507 0.3318>}
 deform_pec_aim_r {parent=deform_pec_aim_base_r, loc=<0 0 0>, rot=<0 0 0 1>}
 deform_pec_base_r {parent=deform_pec_aim_r, loc=<-9.3667 -0.0003 -0.0004>, rot=<0 0 0 1>}
 deform_pec_r {parent=deform_pec_base_r, loc=<0 0 0>, rot=<0 0 0 1>}
 thigh_l {parent=pelvis, loc=<-1.0855 -0.2109 -8.7053>, rot=<0.0009 0.0355 0.0155 -0.9993>}
 calf_l {parent=thigh_l, loc=<-42.5284 -0.0019 -0.7297>, rot=<0.0006 -0.0161 -0.0391 -0.9991>}
 foot_l {parent=calf_l, loc=<-46.7401 -2.5014 -0.592>, rot=<0.0427 -0.0238 0.0353 -0.9982>}
 ball_l {parent=foot_l, loc=<-3.5658 10.8066 -0.1544>, rot=<0.0001 -0.0001 -0.7063 -0.7079>}
 calf_twist_02_l {parent=calf_l, loc=<-19.1546 -1.7943 -0.2638>, rot=<0.0237 0.0095 -0.0035 -0.9997>}
 calf_twist_01_l {parent=calf_l, loc=<-35.8483 -1.9197 -0.5794>, rot=<0.0237 0.0095 -0.0035 -0.9997>}
 deform_knee_in_base_l {parent=calf_l, loc=<0.8031 -5.0231 0.2456>, rot=<0 0 0 -1>}
 deform_knee_in_l {parent=deform_knee_in_base_l, loc=<0 0 0>, rot=<0 0 0 1>}
 deform_knee_out_base_l {parent=calf_l, loc=<0 4 0>, rot=<0 0 0 -1>}
 deform_knee_out_l {parent=deform_knee_out_base_l, loc=<0 0 0>, rot=<0 0 0 1>}
 kneepad_calf_position_l {parent=deform_knee_out_base_l, loc=<0 -4 0>, rot=<0 0 0 -1>}
 base_kneepad_l {parent=kneepad_calf_position_l, loc=<-0.4472 6.3175 0.1545>, rot=<-0.4717 -0.5037 -0.456 -0.562>}
 dyn_kneepad_l {parent=base_kneepad_l, loc=<0 0 0>, rot=<0 0 0 1>}
 base_master_zipper_l {parent=calf_l, loc=<-15.6974 -1.7822 4.8102>, rot=<0.0894 -0.9948 0.0477 0.0052>}
 base_boot_zipper_l {parent=base_master_zipper_l, loc=<0.3654 -0.1043 -0.3671>, rot=<0.0072 -0.056 -0.0024 -0.9984>}
 dyn_boot_zipper_l {parent=base_boot_zipper_l, loc=<0 0 0>, rot=<0 0 0 1>}
 thigh_twist_01_l {parent=thigh_l, loc=<-27.952 0 0>, rot=<0.0013 0 0 -1>}
 deform_groin_base_l {parent=thigh_l, loc=<0.4801 5.6776 -0.2486>, rot=<0 0 0 -1>}
 deform_groin_l {parent=deform_groin_base_l, loc=<0 0 0>, rot=<0 0 0 1>}
 thigh_l_helper {parent=thigh_l, loc=<0 0 0>, rot=<0 0 0 1>}
 thigh_r {parent=pelvis, loc=<-1.0855 -0.2109 8.7053>, rot=<0.0355 -0.0009 0.9993 0.0155>}
 calf_r {parent=thigh_r, loc=<42.5284 0.0019 0.7297>, rot=<0.0006 -0.0161 -0.0391 -0.9991>}
 foot_r {parent=calf_r, loc=<46.7401 2.5014 0.592>, rot=<0.0427 -0.0238 0.0353 -0.9982>}
 ball_r {parent=foot_r, loc=<3.5658 -10.8066 0.1544>, rot=<0.0001 -0.0001 -0.7063 -0.7079>}
 calf_twist_02_r {parent=calf_r, loc=<19.1546 1.7943 0.2638>, rot=<0.0237 0.0095 -0.0035 -0.9997>}
 calf_twist_01_r {parent=calf_r, loc=<35.8483 1.9197 0.5794>, rot=<0.0237 0.0095 -0.0035 -0.9997>}
 deform_knee_in_base_r {parent=calf_r, loc=<-0.8031 5.0231 -0.2457>, rot=<0 0 0 -1>}
 deform_knee_in_r {parent=deform_knee_in_base_r, loc=<0 0 0>, rot=<0 0 0 1>}
 deform_knee_out_base_r {parent=calf_r, loc=<0 -4 0>, rot=<0 0 0 -1>}
 deform_knee_out_r {parent=deform_knee_out_base_r, loc=<0 0 0>, rot=<0 0 0 1>}
 kneepad_calf_position_r {parent=deform_knee_out_base_r, loc=<0 4 0>, rot=<0 0 0 -1>}
 base_kneepad_r {parent=kneepad_calf_position_r, loc=<0.4472 -6.3175 -0.1545>, rot=<-0.456 0.562 0.4717 -0.5037>}
 dyn_kneepad_r {parent=base_kneepad_r, loc=<0 0 0>, rot=<0 0 0 1>}
 base_master_zipper_r {parent=calf_r, loc=<15.6974 1.7821 -4.8102>, rot=<0.9948 0.0894 -0.0052 0.0477>}
 base_boot_zipper_r {parent=base_master_zipper_r, loc=<0.3654 -0.1043 0.3671>, rot=<-0.0072 0.056 -0.0024 -0.9984>}
 dyn_boot_zipper_r {parent=base_boot_zipper_r, loc=<0 0 0>, rot=<0 0 0 1>}
 thigh_twist_01_r {parent=thigh_r, loc=<27.952 0 0>, rot=<0.0013 0 0 -1>}
 deform_groin_base_r {parent=thigh_r, loc=<-0.4801 -5.6776 0.2486>, rot=<0 0 0 -1>}
 deform_groin_r {parent=deform_groin_base_r, loc=<0 0 0>, rot=<0 0 0 1>}
 thigh_r_helper {parent=thigh_r, loc=<0 0 0>, rot=<0 0 0 1>}
 deform_glute_base_l {parent=pelvis, loc=<-7.3466 -9.0213 -9.1245>, rot=<0.0172 0.031 -0.4477 -0.8935>}
 deform_glute_l {parent=deform_glute_base_l, loc=<0 0 0>, rot=<0 0 0 1>}
 deform_glute_base_r {parent=pelvis, loc=<-7.3464 -9.0213 9.1245>, rot=<0.031 -0.0172 0.8935 -0.4477>}
 deform_glute_r {parent=deform_glute_base_r, loc=<0 0 0>, rot=<0 0 0 1>}
 base_front_sleeve_r_0 {parent=pelvis, loc=<3.0207 8.0485 3.0511>, rot=<0.1328 0.6828 0.7163 0.0548>}
 dyn_front_sleeve_r_0 {parent=base_front_sleeve_r_0, loc=<0 0 0>, rot=<0 0 0 1>}
 dyn_front_sleeve_r_1 {parent=dyn_front_sleeve_r_0, loc=<3.7547 0 0>, rot=<0 0 0 1>}
 dyn_front_sleeve_r_2 {parent=dyn_front_sleeve_r_1, loc=<5.3343 0 0>, rot=<0 0 0 1>}
 dyn_front_sleeve_r_3 {parent=dyn_front_sleeve_r_2, loc=<3.8574 0 0>, rot=<0 0 0 1>}
 dyn_front_sleeve_r_4 {parent=dyn_front_sleeve_r_3, loc=<4.2596 0 0>, rot=<0 0 0 1>}
 base_front_sleeve_l_0 {parent=pelvis, loc=<3.3721 8.4848 -2.8995>, rot=<-0.0387 -0.7286 -0.6713 -0.1302>}
 dyn_front_sleeve_l_0 {parent=base_front_sleeve_l_0, loc=<0 0 0>, rot=<0 0 0 1>}
 dyn_front_sleeve_l_1 {parent=dyn_front_sleeve_l_0, loc=<3.8348 0 0>, rot=<0 0 0 1>}
 dyn_front_sleeve_l_2 {parent=dyn_front_sleeve_l_1, loc=<5.5432 0 0>, rot=<0 0 0 1>}
 dyn_front_sleeve_l_3 {parent=dyn_front_sleeve_l_2, loc=<7.843 0 0>, rot=<0 0 0 1>}
 dyn_front_sleeve_l_4 {parent=dyn_front_sleeve_l_3, loc=<5.8054 0 0>, rot=<0 0 0 1>}
 base_buckle {parent=pelvis, loc=<5.0714 10.5881 0>, rot=<0.0127 0.707 0.707 0.0127>}
 dyn_buckle {parent=base_buckle, loc=<0 0 0>, rot=<0 0 0 1>}
 base_wrench {parent=pelvis, loc=<-6.4416 5.5544 -18.8758>, rot=<-0.1898 -0.9033 -0.381 -0.054>}
 dyn_wrench {parent=base_wrench, loc=<0 0 0>, rot=<0 0 0 1>}
 base_screwdriver {parent=pelvis, loc=<-9.2737 1.9473 -21.2027>, rot=<-0.176 -0.9685 -0.1739 -0.0264>}
 dyn_screwdriver {parent=base_screwdriver, loc=<0 0 0>, rot=<0 0 0 1>}
 base_pocket {parent=pelvis, loc=<6.6351 -2.8311 12.8478>, rot=<0.1035 0.5371 0.8263 -0.1347>}
 dyn_pocket {parent=base_pocket, loc=<0 0 0>, rot=<0 0 0 1>}
 base_tissue_0 {parent=pelvis, loc=<1.966 -10.8527 8.212>, rot=<0.1172 -0.5413 0.8306 -0.0577>}
 dyn_tissue_0 {parent=base_tissue_0, loc=<0 0 0>, rot=<0 0.0131 0 -0.9999>}
 dyn_tissue_1 {parent=dyn_tissue_0, loc=<7.8334 0 0>, rot=<0 -0.0884 0 -0.9961>}
 dyn_tissue_2 {parent=dyn_tissue_1, loc=<7.6095 0 0>, rot=<0 -0.0228 0 -0.9997>}
 dyn_tissue_3 {parent=dyn_tissue_2, loc=<6.9869 0 0>, rot=<0 0 0 1>}
 dyn_tissue_4 {parent=dyn_tissue_3, loc=<4.2575 0 0>, rot=<0 0 0 1>}
 base_radio {parent=pelvis, loc=<5.4064 -12.8171 -5.5005>, rot=<0.1303 -0.9752 -0.1614 -0.0765>}
 dyn_radio {parent=base_radio, loc=<0 0 0>, rot=<0 0 0 1>}
 rivet_for_aim_sleeve_r {parent=pelvis, loc=<-22.8407 15.0582 6.1613>, rot=<0.1328 0.6828 0.7163 0.0548>}
 rivet_for_aim_sleeve_l {parent=pelvis, loc=<-26.2699 15.5995 -7.1384>, rot=<-0.0387 -0.7286 -0.6713 -0.1302>}
 rivet_for_aim_tissue {parent=pelvis, loc=<-23.4025 -11.2322 10.8215>, rot=<0.128 -0.5405 0.829 -0.0648>}
 base_jacket_l_0 {parent=pelvis, loc=<-5.3854 -4.5806 -15.5057>, rot=<0.0067 -0.9893 0.0423 -0.1392>}
 dyn_jacket_l_0 {parent=base_jacket_l_0, loc=<0 0 0>, rot=<0 0 0 1>}
 dyn_jacket_l_0_end {parent=dyn_jacket_l_0, loc=<10.1656 0 0>, rot=<0 0 0 1>}
 base_jacket_l_1 {parent=pelvis, loc=<-5.6888 -11.0836 -8.3668>, rot=<-0.0429 -0.8142 0.5781 -0.0322>}
 dyn_jacket_l_1 {parent=base_jacket_l_1, loc=<0 0 0>, rot=<0 0 0 1>}
 dyn_jacket_l_1_end {parent=dyn_jacket_l_1, loc=<11.8722 0 0>, rot=<0 0 0 1>}
 base_jacket_c {parent=pelvis, loc=<-6.0721 -12.5638 -0.0115>, rot=<-0.0103 -0.704 0.71 0.0104>}
 dyn_jacket_c {parent=base_jacket_c, loc=<0 0 0>, rot=<0 0 0 1>}
 dyn_jacket_c_end {parent=dyn_jacket_c, loc=<11.8295 0 0>, rot=<0 0 0 1>}
 base_jacket_r_0 {parent=pelvis, loc=<-5.3854 -4.5806 15.5057>, rot=<0.1392 -0.0423 0.9893 -0.0067>}
 dyn_jacket_r_0 {parent=base_jacket_r_0, loc=<0 0 0>, rot=<0 0 0 1>}
 dyn_jacket_r_0_end {parent=dyn_jacket_r_0, loc=<11.872 0 -0.0001>, rot=<0 0 0 1>}
 base_jacket_r_1 {parent=pelvis, loc=<-5.6888 -11.0836 8.3668>, rot=<0.0322 -0.5781 0.8142 0.0429>}
 dyn_jacket_r_1 {parent=base_jacket_r_1, loc=<0 0 0>, rot=<0 0 0 1>}
 dyn_jacket_r_1_end {parent=dyn_jacket_r_1, loc=<11.872 0 0>, rot=<0 0 0 1>}
 thigh_l_for_cons_offset {parent=pelvis, loc=<-1.0855 -0.2109 -8.7053>, rot=<0.0009 0.0355 0.0155 -0.9992>}
 thigh_l_for_cons {parent=thigh_l_for_cons_offset, loc=<0 0 0>, rot=<0 0 0 1>}
 thigh_l_for_cons_set_rotation {parent=thigh_l_for_cons_offset, loc=<0 0 0>, rot=<0 0 0 1>}
 thigh_r_for_cons_offset {parent=pelvis, loc=<-1.0855 -0.2109 8.7053>, rot=<0.0355 -0.0009 0.9992 0.0155>}
 thigh_r_for_cons {parent=thigh_r_for_cons_offset, loc=<0 0 0>, rot=<0 0 0 1>}
 thigh_r_for_cons_set_rotation {parent=thigh_r_for_cons_offset, loc=<0 0 0>, rot=<0 0 0 1>}
 rivet_for_wrench {parent=pelvis, loc=<-15.6341 5.5396 -18.8465>, rot=<-0.1898 -0.9033 -0.381 -0.054>}
 rivet_for_screwdriver {parent=pelvis, loc=<-17.5244 3.7193 -20.0373>, rot=<-0.176 -0.9685 -0.1739 -0.0264>}
 rivet_for_pocket {parent=pelvis, loc=<-19.566 -5.0974 19.3691>, rot=<0.1035 0.5371 0.8263 -0.1347>}
 base_pocket_off {parent=pelvis, loc=<6.6351 -2.8311 12.8478>, rot=<0.1035 0.5371 0.8263 -0.1347>}
 pocket_for_aim_cons {parent=base_pocket_off, loc=<0 0 0>, rot=<0 0 0 1>}
 base_tissue_2 {parent=pelvis, loc=<1.966 -10.8527 8.212>, rot=<0.1172 -0.5413 0.8306 -0.0577>}
 base_tissue_3 {parent=base_tissue_2, loc=<0 0 0>, rot=<0 0 0 1>}
 attach {parent=root, loc=<0 0 0>, rot=<0 0 0 1>}
 ik_hand_root {parent=root, loc=<0 0 0>, rot=<0 0 0 1>}
 ik_hand_gun {parent=ik_hand_root, loc=<-48.1743 3.3545 104.4381>, rot=<0.6276 0.1309 0.4183 -0.6434>}
 ik_hand_r {parent=ik_hand_gun, loc=<0 0 0>, rot=<0 0 0 1>}
 ik_hand_l {parent=ik_hand_gun, loc=<59.3271 67.6997 34.3524>, rot=<-0.6158 -0.7027 -0.3565 0>}
 ik_foot_root {parent=root, loc=<0 0 0>, rot=<0 0 0 1>}
 ik_foot_l {parent=ik_foot_root, loc=<14.8262 -7.3545 8.296>, rot=<0.0357 0.7045 0.0284 -0.7082>}
 ik_foot_r {parent=ik_foot_root, loc=<-14.8262 -7.3545 8.296>, rot=<0.7082 0.0284 -0.7045 0.0357>}
 root_fp {parent=root, loc=<0 0 0>, rot=<0 0 0 1>}
 attach_fp {parent=root_fp, loc=<0 0 0>, rot=<0 0 0 1>}
 camera_root {parent=root_fp, loc=<0 0 0>, rot=<0 0 0 1>}
 camera {parent=camera_root, loc=<0 0 0>, rot=<0 0 0 1>}
 attach_emote_01 {parent=root, loc=<0 0 0>, rot=<0 0 0 1>}
 attach_emote_02 {parent=root, loc=<0 0 0>, rot=<0 0 0 1>}
 attach_emote_03 {parent=root, loc=<0 0 0>, rot=<0 0 0 1>}
 warp_point_01 {parent=root, loc=<0 0 0>, rot=<0 0 0 1>}
 warp_point_02 {parent=root, loc=<0 0 0>, rot=<0 0 0 1>}
 warp_point_03 {parent=root, loc=<0 0 0>, rot=<0 0 0 1>}
 thigh_r_ctrl {controlType=EULER_TRANSFORM, type=Control, loc=<0 0 0>, rot=<0 0 0 1>}
 calf_r_ctrl {controlType=EULER_TRANSFORM, parent=thigh_r_ctrl, type=Control, loc=<0 0 0>, rot=<0 0 0 1>}
 thigh_l_ctrl {controlType=EULER_TRANSFORM, type=Control, loc=<0 0 0>, rot=<0 0 0 1>}
 calf_l_ctrl {controlType=EULER_TRANSFORM, parent=thigh_l_ctrl, type=Control, loc=<52.9252 -5.276 57.0716>, rot=<-0.059 -0.7294 -0.0004 0.6816>}
]
