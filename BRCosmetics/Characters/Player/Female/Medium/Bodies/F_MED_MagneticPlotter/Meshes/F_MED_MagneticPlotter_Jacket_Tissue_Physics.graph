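# Physics asset for the dangling jacket and tissue bones on the medium female MagneticPlotter player body.
# Five two-link jacket chains and one five-link tissue chain hang from tiny base capsules, with narrow swing limits on the jacket and near-zero twist throughout.
# The tissue chain's first swing axis tightens toward the tip while the other stays wide.
# Unconstrained capsules cover both thighs, the right foot and an oversized, downward-offset root.
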PhysicsAsset /BRCosmetics/Characters/Player/Female/Medium/Bodies/F_MED_MagneticPlotter/Meshes/F_MED_MagneticPlotter_Jacket_Tissue_Physics
bodies {
 base_jacket_c [{shapeType=Capsule, length=0.505, radius=0.505, shapeName=base_jacket_c_capsule, rotation=<rot 0 -90 0>}]
 base_jacket_l_0 [{shapeType=Capsule, length=0.505, radius=0.505, shapeName=base_jacket_l_0_capsule, rotation=<rot 0 -90 0>}]
 base_jacket_l_1 [{shapeType=Capsule, length=0.505, radius=0.505, shapeName=base_jacket_l_1_capsule, rotation=<rot 0 -90 0>}]
 base_jacket_r_0 [{shapeType=Capsule, length=0.505, radius=0.505, shapeName=base_jacket_r_0_capsule, rotation=<rot 0 -90 0>}]
 base_jacket_r_1 [{shapeType=Capsule, length=0.505, radius=0.505, shapeName=base_jacket_r_1_capsule, rotation=<rot 0 -90 0>}]
 base_tissue_0 [{shapeType=Capsule, length=0.505, radius=0.505, shapeName=base_tissue_0_capsule, rotation=<rot 0 -90 0>}]
 dyn_jacket_c [{shapeType=Capsule, length=13.9619, radius=2.3302, shapeName=dyn_jacket_c_end_capsule, center=<5 0 -2.2025>, rotation=<rot 0 0 90>}]
 dyn_jacket_c_end [{shapeType=Capsule, length=13.139, radius=2.1514, shapeName=dyn_jacket_c_end_capsule, center=<5.3726 0 -1.7996>, rotation=<rot 0 0 90>}]
 dyn_jacket_l_0 [{shapeType=Capsule, length=12.9224, radius=2.4336, shapeName=dyn_jacket_l_0_end_capsule, center=<5 0 2.5154>, rotation=<rot 0 0.0001 67.4873>}]
 dyn_jacket_l_0_end [{shapeType=Capsule, length=13.1886, radius=2.1674, shapeName=dyn_jacket_l_0_end_capsule, center=<2 0 0.2835>, rotation=<rot 0 0.0002 65.8045>}]
 dyn_jacket_l_1 [{shapeType=Capsule, length=10.2581, radius=2.3516, shapeName=dyn_jacket_l_1_end_capsule, center=<5 0 -0.7584>, rotation=<rot 0 0 74.2694>}]
 dyn_jacket_l_1_end [{shapeType=Capsule, length=10.2581, radius=2.3516, shapeName=dyn_jacket_l_1_end_capsule, center=<5 0 0>, rotation=<rot 0 0 74.2694>}]
 dyn_jacket_r_0 [{shapeType=Capsule, length=13.8373, radius=2.4336, shapeName=dyn_jacket_r_0_end_capsule, center=<5 0 2.2008>, rotation=<rot 0 0 -79.4436>}]
 dyn_jacket_r_0_end [{shapeType=Capsule, length=14.1035, radius=2.1674, shapeName=dyn_jacket_r_0_end_capsule, center=<2 0 -0.1992>, rotation=<rot 0 0 -79.4436>}]
 dyn_jacket_r_1 [{
   shapeType=Capsule
   length=13.923
   radius=2.2103
   shapeName=dyn_jacket_r_1_end_capsule
   center=<5 0 -0.9779>
   rotation=<rot -5.9117 -1.6713 -74.1831>
  }]
 dyn_jacket_r_1_end [{
   shapeType=Capsule
   length=13.923
   radius=2.2103
   shapeName=dyn_jacket_r_1_end_capsule
   center=<4.0273 0.1748 0.3263>
   rotation=<rot 0 0 -74.2694>
  }]
 dyn_tissue_0 [{shapeType=Capsule, length=10.8601, radius=2.4684, shapeName=dyn_tissue_0_capsule, center=<4 0 2>, rotation=<rot 0 0 111.0125>}]
 dyn_tissue_1 [{shapeType=Capsule, length=10.8601, radius=2.4684, shapeName=dyn_tissue_1_capsule, center=<4 0 2>, rotation=<rot 0 0 111.0125>}]
 dyn_tissue_2 [{shapeType=Capsule, length=10.8601, radius=2.4684, shapeName=dyn_tissue_2_capsule, center=<3 0 2>, rotation=<rot 0 0 111.0125>}]
 dyn_tissue_3 [{shapeType=Capsule, length=10.8601, radius=2.4684, shapeName=dyn_tissue_3_capsule, center=<2.5 0 2>, rotation=<rot 0 0 111.0125>}]
 dyn_tissue_4 [{shapeType=Capsule, length=10.8601, radius=2.4684, shapeName=dyn_tissue_4_capsule, center=<4 0 2>, rotation=<rot 0 0 111.0125>}]
 foot_r [{shapeType=Capsule, length=9.865, radius=8.7824, shapeName=foot_r_capsule, center=<3.4176 -6.793 0>, rotation=<rot 0 0 90>}]
 root [{shapeType=Capsule, length=0.5056, radius=337.6725, shapeName=root_capsule, center=<2.6954 -23.3737 -331.5724>, rotation=<rot 0 -90 90>}]
 thigh_l [{shapeType=Capsule, length=24.9396, radius=8.2254, shapeName=thigh_l_capsule, center=<-18.6445 -0.3384 0>, rotation=<rot 90 0 175.8887>}]
 thigh_r [{shapeType=Capsule, length=24.9396, radius=8.2254, shapeName=thigh_r_capsule, center=<18.6445 0.3384 0>, rotation=<rot 90 0 175.8887>}]
}
constraints [
 {bone1Name=dyn_jacket_c_end, bone2Name=dyn_jacket_c, swing1LimitDegrees=20, swing2LimitDegrees=20, twistLimitDegrees=0}
 {bone1Name=dyn_jacket_l_0_end, bone2Name=dyn_jacket_l_0, swing1LimitDegrees=20, swing2LimitDegrees=20, twistLimitDegrees=0}
 {bone1Name=dyn_jacket_l_1_end, bone2Name=dyn_jacket_l_1, swing1LimitDegrees=20, swing2LimitDegrees=20, twistLimitDegrees=0}
 {bone1Name=dyn_jacket_r_0_end, bone2Name=dyn_jacket_r_0, swing1LimitDegrees=20, swing2LimitDegrees=20, twistLimitDegrees=0}
 {bone1Name=dyn_jacket_r_1_end, bone2Name=dyn_jacket_r_1, swing1LimitDegrees=20, swing2LimitDegrees=20, twistLimitDegrees=0}
 {bone1Name=dyn_jacket_c, bone2Name=base_jacket_c, swing1LimitDegrees=20, swing2LimitDegrees=20, twistLimitDegrees=0}
 {bone1Name=dyn_jacket_l_0, bone2Name=base_jacket_l_0, swing1LimitDegrees=20, swing2LimitDegrees=20, twistLimitDegrees=0}
 {bone1Name=dyn_jacket_l_1, bone2Name=base_jacket_l_1, swing1LimitDegrees=20, swing2LimitDegrees=20, twistLimitDegrees=0}
 {bone1Name=dyn_jacket_r_0, bone2Name=base_jacket_r_0, swing1LimitDegrees=20, swing2LimitDegrees=20, twistLimitDegrees=0}
 {bone1Name=dyn_jacket_r_1, bone2Name=base_jacket_r_1, swing1LimitDegrees=20, swing2LimitDegrees=20, twistLimitDegrees=0}
 {bone1Name=dyn_tissue_0, bone2Name=base_tissue_0, swing1LimitDegrees=20, swing2LimitDegrees=80, twistLimitDegrees=1}
 {bone1Name=dyn_tissue_1, bone2Name=dyn_tissue_0, swing1LimitDegrees=15, swing2LimitDegrees=40, twistLimitDegrees=1}
 {bone1Name=dyn_tissue_2, bone2Name=dyn_tissue_1, swing1LimitDegrees=10, swing2LimitDegrees=50, twistLimitDegrees=1}
 {bone1Name=dyn_tissue_3, bone2Name=dyn_tissue_2, swing1LimitDegrees=5, swing2LimitDegrees=60, twistLimitDegrees=1}
 {bone1Name=dyn_tissue_4, bone2Name=dyn_tissue_3, swing1LimitDegrees=2.5, swing2LimitDegrees=70, twistLimitDegrees=1}
]
