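# Physics asset for the front sleeves of the MagneticPlotter medium female player cosmetic body, apparently for secondary sleeve motion.
# Each side has a five-bone dyn_front_sleeve capsule chain that widens toward the tip, with a zero-size first capsule.
# Constraints link neighboring chain bones with narrow swing limits and zero twist, while the small base sleeve bodies stay unconstrained.
# Large capsules on the thigh helper bones provide leg shapes, likely as colliders for the sleeves.
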PhysicsAsset /BRCosmetics/Characters/Player/Female/Medium/Bodies/F_MED_MagneticPlotter/Meshes/F_MED_MagneticPlotter_Sleeves_Physics
bodies {
 base_front_sleeve_l_0 [{shapeType=Capsule, length=0.505, radius=0.505, shapeName=base_front_sleeve_l_0_capsule, rotation=<rot 0 -90 0>}]
 base_front_sleeve_r_0 [{shapeType=Capsule, length=0.505, radius=0.505, shapeName=base_front_sleeve_r_0_capsule, rotation=<rot 0 -90 0>}]
 dyn_front_sleeve_l_0 [{shapeType=Capsule, length=0, radius=0, shapeName=dyn_front_sleeve_l_0_capsule, center=<2 0 0>, rotation=<rot 0 0 90>}]
 dyn_front_sleeve_l_1 [{shapeType=Capsule, length=5.3846, radius=1.8092, shapeName=dyn_front_sleeve_l_1_capsule, center=<2 -0.5 0>, rotation=<rot 0 0 90>}]
 dyn_front_sleeve_l_2 [{shapeType=Capsule, length=4.7845, radius=2.6248, shapeName=dyn_front_sleeve_l_2_capsule, center=<3 -0.5 0>, rotation=<rot 0 0 90>}]
 dyn_front_sleeve_l_3 [{shapeType=Capsule, length=5.7814, radius=3.1438, shapeName=dyn_front_sleeve_l_3_capsule, center=<2 -0.5 0>, rotation=<rot 0 0 90>}]
 dyn_front_sleeve_l_4 [{shapeType=Capsule, length=5.5773, radius=3.48, shapeName=dyn_front_sleeve_l_3_capsule, center=<4 -0.5 0>, rotation=<rot 0 0 90>}]
 dyn_front_sleeve_r_0 [{shapeType=Capsule, length=0, radius=0, shapeName=dyn_front_sleeve_r_0_capsule, center=<2 0 0>, rotation=<rot 0 0 90>}]
 dyn_front_sleeve_r_1 [{shapeType=Capsule, length=4.5047, radius=1.2161, shapeName=dyn_front_sleeve_r_1_capsule, center=<2 0.5 0>, rotation=<rot 0 0 90>}]
 dyn_front_sleeve_r_2 [{shapeType=Capsule, length=5.7676, radius=2.0481, shapeName=dyn_front_sleeve_r_2_capsule, center=<2 0.5 0>, rotation=<rot 0 0 90>}]
 dyn_front_sleeve_r_3 [{shapeType=Capsule, length=6.1464, radius=2.5434, shapeName=dyn_front_sleeve_r_3_capsule, center=<4 0 0>, rotation=<rot 0 0 90>}]
 dyn_front_sleeve_r_4 [{shapeType=Capsule, length=5.9116, radius=3.2065, shapeName=dyn_front_sleeve_r_4_capsule, center=<6 0 0>, rotation=<rot 0 0 90>}]
 thigh_l_helper [{shapeType=Capsule, length=31.4457, radius=9.0655, shapeName=thigh_l_capsule, center=<-18.4359 -0.9376 0>, rotation=<rot 90 89.4328 -89.2143>}]
 thigh_r_helper [{shapeType=Capsule, length=31.4457, radius=9.0655, shapeName=thigh_l_capsule, center=<18.4699 0.5041 0>, rotation=<rot 90 179.5524 0.9053>}]
}
constraints [
 {bone1Name=dyn_front_sleeve_l_1, bone2Name=dyn_front_sleeve_l_0, swing1LimitDegrees=12, swing2LimitDegrees=25, twistLimitDegrees=0}
 {bone1Name=dyn_front_sleeve_l_2, bone2Name=dyn_front_sleeve_l_1, swing1LimitDegrees=15, swing2LimitDegrees=10, twistLimitDegrees=0}
 {bone1Name=dyn_front_sleeve_l_3, bone2Name=dyn_front_sleeve_l_2, swing1LimitDegrees=20, swing2LimitDegrees=20, twistLimitDegrees=0}
 {bone1Name=dyn_front_sleeve_l_4, bone2Name=dyn_front_sleeve_l_3, swing1LimitDegrees=20, swing2LimitDegrees=20, twistLimitDegrees=0}
 {bone1Name=dyn_front_sleeve_r_1, bone2Name=dyn_front_sleeve_r_0, swing1LimitDegrees=20, swing2LimitDegrees=25, twistLimitDegrees=0}
 {bone1Name=dyn_front_sleeve_r_2, bone2Name=dyn_front_sleeve_r_1, swing1LimitDegrees=20, swing2LimitDegrees=10, twistLimitDegrees=0}
 {bone1Name=dyn_front_sleeve_r_3, bone2Name=dyn_front_sleeve_r_2, swing1LimitDegrees=20, swing2LimitDegrees=10, twistLimitDegrees=0}
 {bone1Name=dyn_front_sleeve_r_4, bone2Name=dyn_front_sleeve_r_3, swing1LimitDegrees=20, swing2LimitDegrees=10, twistLimitDegrees=0}
]
