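# Secondary-motion AnimBlueprint for the female MagneticPlotter character body, layering cloth and accessory physics onto the pose copied from the parent mesh.
# A Control Rig feeds bool-gated jacket, tissue and buckle offsets, then Anim Dynamics and RigidBody passes for accessories, jacket tissue and sleeves.
# Weighted constraints and Look At nodes anchor the hanging wrench and screwdriver.
# The event graph sets gravity, sim-space and blend variables from movement, emote, front-end and zipline states.
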AnimBlueprint /BRCosmetics/Characters/Player/Female/Medium/Bodies/F_MED_MagneticPlotter/Meshes/F_MED_MagneticPlotter_AnimBP
target_skeleton=&/F_MED_MagneticPlotter_Skeleton
graphs [
 AnimGraph {nodes [
   Output Pose(Result=Use cached pose 'dynamicsCache'()) {layerGroup=None}
   Save cached pose 'InCache'(Pose=Copy Pose From Mesh() {bUseAttachedParent=true}) {cacheName=InCache}
   n1 = Control Rig(Source=Use cached pose 'InCache'()) {
    alphaBoolBlend {blendOption=Linear}
    controlRigAssetReference {blueprintRigClass=&/F_MED_MagneticPlotter_CtrlRig.F_MED_MagneticPlotter_CtrlRig_C}
   }
   Save cached pose 'controlRigCache'(Pose=@n1) {cacheName=controlRigCache}
   n2 = Anim Dynamics - Bone: dyn_chain_arm!(
    ComponentPose=Local To Component(LocalPose=Use cached pose 'sleevesDyn'())
    Alpha=1
    ComponentLinearAccScale:Vector="0.3,0.3,0.3"
    ComponentLinearVelScale:Vector="0.3,0.3,0.3"
   ) -> (Pose) {
    angularDampingOverride=1
    angularSpringConstant=5000
    bAngularSpring=true
    bOverrideAngularDamping=true
    bUseGravityOverride=true
    gravityScale=0.35
    simulationSpace=RootRelative
    alphaBoolBlend {blendOption=Linear}
    boundBone {boneName=dyn_chain_arm}
    componentAppliedLinearAccClamp=<3000 3000 3000>
    componentLinearAccScale=<0.3 0.3 0.3>
    componentLinearVelScale=<0.3 0.3 0.3>
    gravityOverride {z=-980}
    physicsBodyDefinitions [{
      boundBone {boneName=dyn_chain_arm}
      boxExtents=<22 22 22>
      constraintSetup {angularLimitsMax=<5 0 15>, angularLimitsMin=<-5 0 -17>, angularTarget=<1 0 0>}
      localJointOffset=<8 0 0>
     }]
   }
   Save cached pose 'dynamicsCache'(Pose=Component To Local(ComponentPose=@n2)) {cacheName=dynamicsCache}
   n3 = "Transform (Modify) Bone - Bone: base_jacket_l_0"!(
    ComponentPose=Local To Component(LocalPose=Use cached pose 'controlRigCache'())
    bAlphaBoolEnabled:bool=True
    Translation="5,0,0"
    Rotation="-17.165,38.058,-16.857"
   ) -> (Pose) {
    alphaInputType=Bool
    rotationMode=BMM_Additive
    rotationSpace=BCS_BoneSpace
    scaleSpace=BCS_BoneSpace
    translationMode=BMM_Additive
    translationSpace=BCS_BoneSpace
    alphaBoolBlend {blendInTime=0.1, blendOption=Linear, blendOutTime=0.1}
    boneToModify {boneName=base_jacket_l_0}
    rotation=<rot -17.165 38.058 -16.857>
    translation {x=5}
   }
   n4 = "Transform (Modify) Bone - Bone: base_jacket_l_1"!(ComponentPose=@n3, bAlphaBoolEnabled:bool=True, Rotation="-18.877,0.056,-0.015") -> (Pose) {
    alphaInputType=Bool
    rotationMode=BMM_Additive
    rotationSpace=BCS_BoneSpace
    scaleSpace=BCS_BoneSpace
    translationMode=BMM_Additive
    translationSpace=BCS_BoneSpace
    alphaBoolBlend {blendInTime=0.1, blendOption=Linear, blendOutTime=0.1}
    boneToModify {boneName=base_jacket_l_1}
    rotation=<rot -18.877 0.056 -0.015>
   }
   n5 = "Transform (Modify) Bone - Bone: base_jacket_c"!(ComponentPose=@n4, bAlphaBoolEnabled:bool=True, Rotation="-30,0,0") -> (Pose) {
    alphaInputType=Bool
    rotationMode=BMM_Additive
    rotationSpace=BCS_BoneSpace
    alphaBoolBlend {blendInTime=0.1, blendOption=Linear, blendOutTime=0.1}
    boneToModify {boneName=base_jacket_c}
    rotation {pitch=-30}
   }
   n6 = "Transform (Modify) Bone - Bone: base_jacket_l_1"(ComponentPose=@n5, bAlphaBoolEnabled=True, Rotation="-30,0,30") {
    alphaInputType=Bool
    rotationMode=BMM_Additive
    rotationSpace=BCS_BoneSpace
    alphaBoolBlend {blendInTime=0.1, blendOption=Linear, blendOutTime=0.1}
    boneToModify {boneName=base_jacket_l_1}
    rotation {pitch=-30, roll=30}
   }
   n7 = "Transform (Modify) Bone - Bone: base_jacket_r_1"!(ComponentPose=@n6, bAlphaBoolEnabled:bool=True, Rotation="-30,0,-30") -> (Pose) {
    alphaInputType=Bool
    rotationMode=BMM_Additive
    rotationSpace=BCS_BoneSpace
    alphaBoolBlend {blendInTime=0.1, blendOption=Linear, blendOutTime=0.1}
    boneToModify {boneName=base_jacket_r_1}
    rotation {pitch=-30, roll=-30}
   }
   n8 = "Transform (Modify) Bone - Bone: base_tissue_0"!(ComponentPose=@n7, bAlphaBoolEnabled:bool=True, Translation="0,0,0", Rotation="20,0,0", Scale="1,1,1") -> (Pose) {
    alphaInputType=Bool
    rotationMode=BMM_Additive
    rotationSpace=BCS_BoneSpace
    alphaBoolBlend {blendInTime=1, blendOption=Linear}
    boneToModify {boneName=base_tissue_0}
    rotation {pitch=20}
   }
   n9 = "Transform (Modify) Bone - Bone: base_tissue_0"(ComponentPose=@n8, bAlphaBoolEnabled=True, Translation="0,0,0", Rotation="90,0,-30", Scale="1,1,1") {
    alphaInputType=Bool
    rotationMode=BMM_Additive
    rotationSpace=BCS_BoneSpace
    alphaBoolBlend {blendInTime=0.5, blendOption=Linear, blendOutTime=0.5}
    boneToModify {boneName=base_tissue_0}
    rotation {pitch=90, roll=-30}
   }
   n10 = "Transform (Modify) Bone - Bone: base_tissue_0"(ComponentPose=@n9, bAlphaBoolEnabled=True, Translation="-5,0,0", Rotation="140,0,0", Scale="1,1,1") {
    alphaInputType=Bool
    rotationMode=BMM_Additive
    rotationSpace=BCS_BoneSpace
    translationMode=BMM_Additive
    translationSpace=BCS_BoneSpace
    alphaBoolBlend {blendInTime=0.5, blendOption=Linear}
    boneToModify {boneName=base_tissue_0}
    rotation {pitch=140}
    translation {x=-5}
   }
   n11 = "Transform (Modify) Bone - Bone: base_tissue_0"!(ComponentPose=@n10, bAlphaBoolEnabled=True, Rotation="80,0,0") -> (Pose) {
    alphaInputType=Bool
    rotationMode=BMM_Additive
    rotationSpace=BCS_BoneSpace
    translationSpace=BCS_BoneSpace
    alphaBoolBlend {blendInTime=0.7, blendOption=QuadraticInOut, blendOutTime=0.35}
    boneToModify {boneName=base_tissue_0}
    rotation {pitch=80}
   }
   n12 = "Transform (Modify) Bone - Bone: base_buckle"!(ComponentPose=@n11, bAlphaBoolEnabled:bool=True, Translation="1,0.5,0.5", Rotation="-8.09,0.002,0", Scale="0.97,0.97,0.97") -> (Pose) {
    alphaInputType=Bool
    rotationMode=BMM_Additive
    rotationSpace=BCS_BoneSpace
    scaleMode=BMM_Additive
    scaleSpace=BCS_BoneSpace
    translationMode=BMM_Additive
    translationSpace=BCS_BoneSpace
    alphaBoolBlend {blendOption=Linear}
    boneToModify {boneName=base_buckle}
    rotation {pitch=-8.09, yaw=0.002}
    scale=<0.97 0.97 0.97>
    translation=<1 0.5 0.5>
   }
   Save cached pose 'transformsModifyCache'(Pose=Component To Local(ComponentPose=@n12)) {cacheName=transformsModifyCache}
   n13 = Anim Dynamics - Bone: dyn_boot_zipper_l!(
    ComponentPose=Local To Component(LocalPose=Use cached pose 'transformsModifyCache'())
    Alpha=1
    ComponentLinearAccScale:Vector="0,0,0"
    ComponentLinearVelScale:Vector="0,0,0"
   ) -> (Pose) {
    angularDampingOverride=1
    angularSpringConstant=20000
    bAngularSpring=true
    bOverrideAngularDamping=true
    bUseGravityOverride=true
    gravityScale=0.35
    simulationSpace=RootRelative
    alphaBoolBlend {blendOption=Linear}
    boundBone {boneName=dyn_boot_zipper_l}
    componentAppliedLinearAccClamp=<3000 3000 3000>
    gravityOverride {z=-980}
    physicsBodyDefinitions [{
      boundBone {boneName=dyn_boot_zipper_l}
      boxExtents=<20 25 25>
      constraintSetup {angularLimitsMax=<0 30 30>, angularLimitsMin=<0 -15 -30>, angularTarget=<1 0 0>}
      localJointOffset=<8 0 0>
     }]
   }
   n14 = Anim Dynamics - Bone: dyn_boot_zipper_r!(ComponentPose=@n13, Alpha=1, ComponentLinearAccScale:Vector="0,0,0", ComponentLinearVelScale:Vector="0,0,0") -> (Pose) {
    angularDampingOverride=1
    angularSpringConstant=20000
    bAngularSpring=true
    bOverrideAngularDamping=true
    bUseGravityOverride=true
    gravityScale=0.35
    simulationSpace=RootRelative
    alphaBoolBlend {blendOption=Linear}
    boundBone {boneName=dyn_boot_zipper_r}
    componentAppliedLinearAccClamp=<3000 3000 3000>
    gravityOverride {z=-980}
    physicsBodyDefinitions [{
      boundBone {boneName=dyn_boot_zipper_r}
      boxExtents=<20 25 25>
      constraintSetup {angularLimitsMax=<0 15 30>, angularLimitsMin=<0 -30 -30>, angularTarget=<1 0 0>}
      localJointOffset=<8 0 0>
     }]
   }
   n15 = Anim Dynamics - Bone: dyn_radio!(ComponentPose=@n14, Alpha=1, ComponentLinearAccScale:Vector="0,0,0", ComponentLinearVelScale:Vector="0,0,0") -> (Pose) {
    angularDampingOverride=1
    angularSpringConstant=5000
    bAngularSpring=true
    bOverrideAngularDamping=true
    bUseGravityOverride=true
    gravityScale=0.35
    simulationSpace=RootRelative
    alphaBoolBlend {blendOption=Linear}
    boundBone {boneName=dyn_radio}
    componentAppliedLinearAccClamp=<3000 3000 3000>
    gravityOverride {z=-980}
    physicsBodyDefinitions [{
      boundBone {boneName=dyn_radio}
      boxExtents=<20 20 20>
      constraintSetup {angularLimitsMax=<0 3 5>, angularLimitsMin=<0 -3 -5>, angularTarget=<1 0 0>}
      localJointOffset=<5 0 0>
     }]
   }
   n16 = Anim Dynamics - Bone: dyn_pocket!(ComponentPose=@n15, Alpha=1, ComponentLinearAccScale:Vector="0,0,0", ComponentLinearVelScale:Vector="0,0,0") -> (Pose) {
    angularDampingOverride=1
    angularSpringConstant=20000
    bAngularSpring=true
    bOverrideAngularDamping=true
    bUseGravityOverride=true
    gravityScale=0.35
    simulationSpace=RootRelative
    alphaBoolBlend {blendOption=Linear}
    boundBone {boneName=dyn_pocket}
    componentAppliedLinearAccClamp=<3000 3000 3000>
    gravityOverride {z=-980}
    physicsBodyDefinitions [{
      boundBone {boneName=dyn_pocket}
      boxExtents=<30 20 20>
      constraintSetup {angularLimitsMax=<0 5 10>, angularLimitsMin=<0 -5 0>, angularTarget=<1 0 0>}
      localJointOffset=<5 0 0>
     }]
   }
   Save cached pose 'dynPocketCache'(Pose=Component To Local(ComponentPose=@n16)) {cacheName=dynPocketCache}
   n17 = RigidBody!(
    ComponentPose=Local To Component(LocalPose=Use cached pose 'dynPocketCache'())
    Alpha
    OverrideWorldGravity:Vector="0,0,0"
    SimSpaceSettings:Sim Space Settings Structure=(WorldAlpha=0,VelocityScaleZ=1,DampingAlpha=1,MaxLinearVelocity=10000,MaxAngularVelocity=10000,MaxLinearAcceleration=10000,MaxAngularAcceleration=10000,ExternalLinearDragV=(X=0,Y=0,Z=0),ExternalLinearVelocity=(X=0,Y=0,Z=0),ExternalAngularVelocity=(X=0,Y=0,Z=0))
   ) -> (Pose) {
    bClampLinearTranslationLimitToRefPose=true
    bOverrideWorldGravity=true
    overridePhysicsAsset=&/F_MED_MagneticPlotter_Jacket_Tissue_Physics
    simulationSpace=BaseBoneSpace
    alphaBoolBlend {blendOption=Linear}
    baseBoneRef {boneName=pelvis}
    simSpaceSettings {
     maxAngularAcceleration=500
     maxAngularVelocity=3
     maxLinearAcceleration=400
     maxLinearVelocity=250
     velocityScaleZ=0.5
     worldAlpha=0.2
     externalLinearDragV=<0.1 0.1>
    }
   }
   n18 = Anim Dynamics - Bone: base_tissue_0!(
    ComponentPose=@n17
    Alpha=0
    GravityOverride:Vector="0,0,-980"
    ComponentLinearAccScale:Vector="0,0,0"
    ComponentLinearVelScale:Vector="0,0,0"
   ) -> (Pose) {
    angularDampingOverride=1
    angularSpringConstant=1000
    bOverrideAngularDamping=true
    gravityScale=0.35
    simulationSpace=RootRelative
    alphaBoolBlend {blendInTime=0.1, blendOption=Linear, blendOutTime=0.1}
    alphaScaleBiasClamp {bInterpResult=true, interpSpeedDecreasing=1, interpSpeedIncreasing=5}
    boundBone {boneName=base_tissue_0}
    componentAppliedLinearAccClamp=<3000 3000 3000>
    gravityOverride {z=-980}
    physicsBodyDefinitions [{
      boundBone {boneName=base_tissue_0}
      boxExtents=<50 50 50>
      constraintSetup {angularLimitsMax=<0 0 20>, angularLimitsMin=<0 -30 0>, angularTarget=<1 0 0>}
      localJointOffset=<6.8 0 0>
     }]
    relativeSpaceBone {boneName=pelvis}
   }
   n19 = Anim Dynamics - Bone: dyn_tissue_1!(
    ComponentPose=@n18
    Alpha=1
    GravityOverride:Vector="0,0,-980"
    ComponentLinearAccScale:Vector="0,0,0"
    ComponentLinearVelScale:Vector="0,0,0"
   ) -> (Pose) {
    angularDampingOverride=1
    angularSpringConstant=20000
    bOverrideAngularDamping=true
    gravityScale=0.35
    simulationSpace=RootRelative
    alphaBoolBlend {blendInTime=0.1, blendOption=Linear, blendOutTime=0.1}
    alphaScaleBiasClamp {bInterpResult=true, interpSpeedDecreasing=1, interpSpeedIncreasing=5}
    boundBone {boneName=dyn_tissue_1}
    componentAppliedLinearAccClamp=<3000 3000 3000>
    gravityOverride {z=-980}
    physicsBodyDefinitions [{
      boundBone {boneName=dyn_tissue_1}
      boxExtents=<40 40 40>
      constraintSetup {angularLimitsMax=<0 0 30>, angularLimitsMin=<0 -30 -10>, angularTarget=<1 0 0>}
      localJointOffset=<8 0 0>
     }]
    relativeSpaceBone {boneName=pelvis}
   }
   n20 = Anim Dynamics - Bone: dyn_tissue_2!(
    ComponentPose=@n19
    Alpha=1
    GravityOverride:Vector="0,0,-980"
    ComponentLinearAccScale:Vector="0,0,0"
    ComponentLinearVelScale:Vector="0,0,0"
   ) -> (Pose) {
    angularDampingOverride=1
    angularSpringConstant=20000
    bAngularSpring=true
    bOverrideAngularDamping=true
    gravityScale=0.35
    simulationSpace=RootRelative
    alphaBoolBlend {blendInTime=0.1, blendOption=Linear, blendOutTime=0.1}
    alphaScaleBiasClamp {bInterpResult=true, interpSpeedDecreasing=1, interpSpeedIncreasing=5}
    boundBone {boneName=dyn_tissue_2}
    componentAppliedLinearAccClamp=<3000 3000 3000>
    gravityOverride {z=-980}
    physicsBodyDefinitions [{
      boundBone {boneName=dyn_tissue_2}
      boxExtents=<40 40 40>
      constraintSetup {angularLimitsMax=<0 0 15>, angularLimitsMin=<0 -15 -10>, angularTarget=<1 0 0>}
      localJointOffset=<8 0 0>
     }]
    relativeSpaceBone {boneName=pelvis}
   }
   n21 = Anim Dynamics - Bone: dyn_tissue_3!(
    ComponentPose=@n20
    Alpha=1
    GravityOverride:Vector="0,0,-980"
    ComponentLinearAccScale:Vector="0,0,0"
    ComponentLinearVelScale:Vector="0,0,0"
   ) -> (Pose) {
    angularDampingOverride=1
    angularSpringConstant=20000
    bAngularSpring=true
    bOverrideAngularDamping=true
    gravityScale=0.35
    simulationSpace=RootRelative
    alphaBoolBlend {blendInTime=0.1, blendOption=Linear, blendOutTime=0.1}
    alphaScaleBiasClamp {bInterpResult=true, interpSpeedDecreasing=1, interpSpeedIncreasing=5}
    boundBone {boneName=dyn_tissue_3}
    componentAppliedLinearAccClamp=<3000 3000 3000>
    gravityOverride {z=-980}
    physicsBodyDefinitions [{
      boundBone {boneName=dyn_tissue_3}
      boxExtents=<40 40 40>
      constraintSetup {angularLimitsMax=<0 0 10>, angularLimitsMin=<0 -10 -5>, angularTarget=<1 0 0>}
      localJointOffset=<8 0 0>
     }]
    relativeSpaceBone {boneName=pelvis}
   }
   Save cached pose 'tissueDyn'(Pose=Component To Local(ComponentPose=@n21)) {cacheName=tissueDyn}
   n22 = "Constraint - base_screwdriver (PPPPPPPPP)"!(
    ComponentPose=Local To Component(LocalPose=Use cached pose 'tissueDyn'())
    bAlphaBoolEnabled:bool=True
    ConstraintWeights_0=0.3316
    ConstraintWeights_1:float=0.1941
    ConstraintWeights_2:float=0.1696
    ConstraintWeights_3:float=0.1486
    ConstraintWeights_4:float=0.1067
    ConstraintWeights_5:float=0.0218
    ConstraintWeights_6:float=0.021
    ConstraintWeights_7:float=0.0064
    ConstraintWeights_8:float=0.0002
   ) -> (Pose) {
    alphaInputType=Bool
    alphaBoolBlend {blendOption=Linear}
    boneToModify {boneName=base_screwdriver}
    constraintSetup [
     {targetBone {boneName=thigh_twist_01_l}}
     {targetBone {boneName=thigh_l}}
     {targetBone {boneName=dyn_jacket_l_0}}
     {targetBone {boneName=pelvis}}
     {targetBone {boneName=dyn_jacket_l_0_end}}
     {targetBone {boneName=deform_groin_l}}
     {targetBone {boneName=dyn_jacket_l_1}}
     {targetBone {boneName=dyn_jacket_l_1_end}}
     {targetBone {boneName=deform_glute_l}}
    ]
    constraintWeights [1,1,1,1,1,1,1,1,1]
   }
   n23 = "Constraint - base_wrench (PPPPPPPPP)"!(
    ComponentPose=@n22
    bAlphaBoolEnabled:bool=True
    ConstraintWeights_0=0.3615
    ConstraintWeights_1:float=0.2599
    ConstraintWeights_2:float=0.1966
    ConstraintWeights_3:float=0.0956
    ConstraintWeights_4:float=0.0445
    ConstraintWeights_5:float=0.0358
    ConstraintWeights_6:float=0.0039
    ConstraintWeights_7:float=0.002
    ConstraintWeights_8:float=0.0002
   ) -> (Pose) {
    alphaInputType=Bool
    alphaBoolBlend {blendOption=Linear}
    boneToModify {boneName=base_wrench}
    constraintSetup [
     {targetBone {boneName=thigh_twist_01_l}}
     {targetBone {boneName=thigh_l}}
     {targetBone {boneName=pelvis}}
     {targetBone {boneName=dyn_jacket_l_0}}
     {targetBone {boneName=deform_groin_l}}
     {targetBone {boneName=dyn_jacket_l_0_end}}
     {targetBone {boneName=dyn_jacket_l_1}}
     {targetBone {boneName=deform_glute_l}}
     {targetBone {boneName=spine_02}}
    ]
    constraintWeights [1,1,1,1,1,1,1,1,1]
   }
   n24 = "Constraint - rivet_for_wrench (PPPPPPPP)"!(
    ComponentPose=@n23
    bAlphaBoolEnabled:bool=True
    ConstraintWeights_0=0.6959
    ConstraintWeights_1:float=0.1148
    ConstraintWeights_2:float=0.0827
    ConstraintWeights_3:float=0.0448
    ConstraintWeights_4:float=0.0382
    ConstraintWeights_5:float=0.0209
    ConstraintWeights_6:float=0.0014
    ConstraintWeights_7:float=0.0012
   ) -> (Pose) {
    alphaInputType=Bool
    alphaBoolBlend {blendOption=Linear}
    boneToModify {boneName=rivet_for_wrench}
    constraintSetup [
     {targetBone {boneName=thigh_twist_01_l}}
     {targetBone {boneName=thigh_l}}
     {targetBone {boneName=dyn_jacket_l_0_end}}
     {targetBone {boneName=dyn_jacket_l_0}}
     {targetBone {boneName=pelvis}}
     {targetBone {boneName=deform_groin_l}}
     {targetBone {boneName=dyn_jacket_l_1}}
     {targetBone {boneName=dyn_jacket_l_1_end}}
    ]
    constraintWeights [1,1,1,1,1,1,1,1]
   }
   n25 = "Constraint - rivet_for_screwdriver (PPPPPPPP)"!(
    ComponentPose=@n24
    bAlphaBoolEnabled:bool=True
    ConstraintWeights_0=0.6884
    ConstraintWeights_1:float=0.1348
    ConstraintWeights_2:float=0.0805
    ConstraintWeights_3:float=0.0441
    ConstraintWeights_4:float=0.0257
    ConstraintWeights_5:float=0.0191
    ConstraintWeights_6:float=0.0041
    ConstraintWeights_7:float=0.0033
   ) -> (Pose) {
    alphaInputType=Bool
    alphaBoolBlend {blendOption=Linear}
    boneToModify {boneName=rivet_for_screwdriver}
    constraintSetup [
     {targetBone {boneName=thigh_twist_01_l}}
     {targetBone {boneName=dyn_jacket_l_0_end}}
     {targetBone {boneName=thigh_l}}
     {targetBone {boneName=dyn_jacket_l_0}}
     {targetBone {boneName=pelvis}}
     {targetBone {boneName=deform_groin_l}}
     {targetBone {boneName=dyn_jacket_l_1_end}}
     {targetBone {boneName=dyn_jacket_l_1}}
    ]
    constraintWeights [1,1,1,1,1,1,1,1]
   }
   n26 = "Constraint - base_screwdriver (PPPPPPPP)"!(
    ComponentPose=@n25
    bAlphaBoolEnabled:bool=True
    ConstraintWeights_0=0.5257
    ConstraintWeights_1:float=0.1696
    ConstraintWeights_2:float=0.1486
    ConstraintWeights_3:float=0.1067
    ConstraintWeights_4:float=0.0218
    ConstraintWeights_5:float=0.021
    ConstraintWeights_6:float=0.0064
    ConstraintWeights_7:float=0.0002
   ) -> (Pose) {
    alphaInputType=Bool
    alphaBoolBlend {blendOption=Linear}
    boneToModify {boneName=base_screwdriver}
    constraintSetup [
     {targetBone {boneName=thigh_l}}
     {targetBone {boneName=dyn_jacket_l_0}}
     {targetBone {boneName=pelvis}}
     {targetBone {boneName=dyn_jacket_l_0_end}}
     {targetBone {boneName=deform_groin_l}}
     {targetBone {boneName=dyn_jacket_l_1}}
     {targetBone {boneName=dyn_jacket_l_1_end}}
     {targetBone {boneName=deform_glute_l}}
    ]
    constraintWeights [1,1,1,1,1,1,1,1]
   }
   n27 = "Constraint - base_wrench (PPPPPPPP)"!(
    ComponentPose=@n26
    bAlphaBoolEnabled:bool=True
    ConstraintWeights_0=0.6214
    ConstraintWeights_1:float=0.1966
    ConstraintWeights_2:float=0.0956
    ConstraintWeights_3:float=0.0445
    ConstraintWeights_4:float=0.0358
    ConstraintWeights_5:float=0.0039
    ConstraintWeights_6:float=0.002
    ConstraintWeights_7:float=0.0002
   ) -> (Pose) {
    alphaInputType=Bool
    alphaBoolBlend {blendOption=Linear}
    boneToModify {boneName=base_wrench}
    constraintSetup [
     {targetBone {boneName=thigh_l}}
     {targetBone {boneName=pelvis}}
     {targetBone {boneName=dyn_jacket_l_0}}
     {targetBone {boneName=deform_groin_l}}
     {targetBone {boneName=dyn_jacket_l_0_end}}
     {targetBone {boneName=dyn_jacket_l_1}}
     {targetBone {boneName=deform_glute_l}}
     {targetBone {boneName=spine_02}}
    ]
    constraintWeights [1,1,1,1,1,1,1,1]
   }
   n28 = "Constraint - rivet_for_wrench (PPPPPPP)"!(
    ComponentPose=@n27
    bAlphaBoolEnabled:bool=True
    ConstraintWeights_0=0.8107
    ConstraintWeights_1:float=0.0827
    ConstraintWeights_2:float=0.0448
    ConstraintWeights_3:float=0.0382
    ConstraintWeights_4:float=0.0209
    ConstraintWeights_5:float=0.0014
    ConstraintWeights_6:float=0.0012
   ) -> (Pose) {
    alphaInputType=Bool
    alphaBoolBlend {blendOption=Linear}
    boneToModify {boneName=rivet_for_wrench}
    constraintSetup [
     {targetBone {boneName=thigh_l}}
     {targetBone {boneName=dyn_jacket_l_0_end}}
     {targetBone {boneName=dyn_jacket_l_0}}
     {targetBone {boneName=pelvis}}
     {targetBone {boneName=deform_groin_l}}
     {targetBone {boneName=dyn_jacket_l_1}}
     {targetBone {boneName=dyn_jacket_l_1_end}}
    ]
    constraintWeights [1,1,1,1,1,1,1]
   }
   n29 = "Constraint - rivet_for_screwdriver (PPPPPPP)"!(
    ComponentPose=@n28
    bAlphaBoolEnabled:bool=True
    ConstraintWeights_0=0.1348
    ConstraintWeights_1:float=0.769
    ConstraintWeights_2:float=0.0441
    ConstraintWeights_3:float=0.0257
    ConstraintWeights_4:float=0.0191
    ConstraintWeights_5:float=0.0041
    ConstraintWeights_6:float=0.0033
   ) -> (Pose) {
    alphaInputType=Bool
    alphaBoolBlend {blendOption=Linear}
    boneToModify {boneName=rivet_for_screwdriver}
    constraintSetup [
     {targetBone {boneName=dyn_jacket_l_0_end}}
     {targetBone {boneName=thigh_l}}
     {targetBone {boneName=dyn_jacket_l_0}}
     {targetBone {boneName=pelvis}}
     {targetBone {boneName=deform_groin_l}}
     {targetBone {boneName=dyn_jacket_l_1_end}}
     {targetBone {boneName=dyn_jacket_l_1}}
    ]
    constraintWeights [1,1,1,1,1,1,1]
   }
   n30 = "Transform (Modify) Bone - Bone: rivet_for_wrench"!(ComponentPose=@n29, Alpha=1, Translation="0,0,28") -> (Pose) {
    translationMode=BMM_Additive
    translationSpace=BCS_BoneSpace
    alphaBoolBlend {blendOption=Linear}
    boneToModify {boneName=rivet_for_wrench}
    translation {z=28}
   }
   n31 = "Transform (Modify) Bone - Bone: rivet_for_screwdriver"!(ComponentPose=@n30, Alpha=1, Translation="0,10,30") -> (Pose) {
    translationMode=BMM_Additive
    translationSpace=BCS_BoneSpace
    alphaBoolBlend {blendOption=Linear}
    boneToModify {boneName=rivet_for_screwdriver}
    translation {y=10, z=30}
   }
   n32 = Look At - Bone: base_screwdriver(ComponentPose=@n31, Alpha=1) {
    alphaBoolBlend {blendOption=Linear}
    boneToModify {boneName=base_screwdriver}
    lookAtTarget {boneReference {boneName=rivet_for_screwdriver}}
    lookAt_Axis {axis=<1 0>}
    lookUp_Axis {axis {x=0, z=1}}
   }
   n33 = Look At - Bone: base_wrench(ComponentPose=@n32, Alpha=1) {
    alphaBoolBlend {blendOption=Linear}
    boneToModify {boneName=base_wrench}
    lookAtTarget {boneReference {boneName=rivet_for_wrench}}
    lookAt_Axis {axis=<1 0>}
    lookUp_Axis {axis {x=0, z=1}}
   }
   n34 = Anim Dynamics - Bone: dyn_wrench!(ComponentPose=@n33, Alpha=1, ComponentLinearAccScale:Vector="0.3,0.3,0.3", ComponentLinearVelScale:Vector="0.3,0.3,0.3") -> (Pose) {
    angularDampingOverride=1
    angularSpringConstant=20000
    bAngularSpring=true
    bOverrideAngularDamping=true
    bUseGravityOverride=true
    gravityScale=0.35
    simulationSpace=RootRelative
    alphaBoolBlend {blendOption=Linear}
    boundBone {boneName=dyn_wrench}
    componentAppliedLinearAccClamp=<3000 3000 3000>
    componentLinearAccScale=<0.3 0.3 0.3>
    componentLinearVelScale=<0.3 0.3 0.3>
    gravityOverride {z=-980}
    physicsBodyDefinitions [{
      boundBone {boneName=dyn_wrench}
      boxExtents=<25 30 30>
      constraintSetup {angularLimitsMax=<0 3 10>, angularLimitsMin=<0 -3 -10>, angularTarget=<1 0 0>}
      localJointOffset=<8 0 0>
     }]
   }
   n35 = Anim Dynamics - Bone: dyn_screwdriver!(ComponentPose=@n34, Alpha=1, ComponentLinearAccScale:Vector="0.3,0.3,0.3", ComponentLinearVelScale:Vector="0.3,0.3,0.3") -> (Pose) {
    angularDampingOverride=1
    angularSpringConstant=20000
    bAngularSpring=true
    bOverrideAngularDamping=true
    bUseGravityOverride=true
    gravityScale=0.35
    simulationSpace=RootRelative
    alphaBoolBlend {blendOption=Linear}
    boundBone {boneName=dyn_screwdriver}
    componentAppliedLinearAccClamp=<3000 3000 3000>
    componentLinearAccScale=<0.3 0.3 0.3>
    componentLinearVelScale=<0.3 0.3 0.3>
    gravityOverride {z=-980}
    physicsBodyDefinitions [{
      boundBone {boneName=dyn_screwdriver}
      boxExtents=<25 30 30>
      constraintSetup {angularLimitsMax=<0 5 8>, angularLimitsMin=<0 -5 -8>, angularTarget=<1 0 0>}
      localJointOffset=<8 0 0>
     }]
   }
   Save cached pose 'toolsDyn'(Pose=Component To Local(ComponentPose=@n35)) {cacheName=toolsDyn}
   n36 = "Transform (Modify) Bone - Bone: thigh_l_helper"!(
    ComponentPose=Local To Component(LocalPose=Use cached pose 'toolsDyn'())
    bAlphaBoolEnabled:bool=True
    Translation="0,-4,0"
    Rotation="0,-5,0"
   ) -> (Pose) {
    alphaInputType=Bool
    rotationMode=BMM_Additive
    rotationSpace=BCS_BoneSpace
    translationMode=BMM_Additive
    translationSpace=BCS_BoneSpace
    alphaBoolBlend {blendOption=Linear}
    boneToModify {boneName=thigh_l_helper}
    rotation {yaw=-5}
    translation {y=-4}
   }
   n37 = "Transform (Modify) Bone - Bone: thigh_r_helper"!(ComponentPose=@n36, bAlphaBoolEnabled:bool=True, Translation="0,4,0", Rotation="0,-5,0") -> (Pose) {
    alphaInputType=Bool
    rotationMode=BMM_Additive
    rotationSpace=BCS_BoneSpace
    translationMode=BMM_Additive
    translationSpace=BCS_BoneSpace
    alphaBoolBlend {blendOption=Linear}
    boneToModify {boneName=thigh_r_helper}
    rotation {yaw=-5}
    translation {y=4}
   }
   n38 = "Transform (Modify) Bone - Bone: base_front_sleeve_l_0"!(ComponentPose=@n37, bAlphaBoolEnabled:bool=True, Translation="0.184,-0.090,-0.649", Rotation="17.628,2.816,8.379") -> (Pose) {
    alphaInputType=Bool
    rotationMode=BMM_Additive
    rotationSpace=BCS_BoneSpace
    translationMode=BMM_Additive
    translationSpace=BCS_BoneSpace
    alphaBoolBlend {blendOption=Linear}
    boneToModify {boneName=base_front_sleeve_l_0}
    rotation=<rot 17.628 2.816 8.379>
    translation=<0.184 -0.09 -0.649>
   }
   n39 = RigidBody(
    ComponentPose=@n38
    OverrideWorldGravity="0,0,0"
    SimSpaceSettings=(WorldAlpha=0,VelocityScaleZ=1,DampingAlpha=1,MaxLinearVelocity=10000,MaxAngularVelocity=10000,MaxLinearAcceleration=10000,MaxAngularAcceleration=10000,ExternalLinearDragV=(X=0,Y=0,Z=0),ExternalLinearVelocity=(X=0,Y=0,Z=0),ExternalAngularVelocity=(X=0,Y=0,Z=0))
   ) {
    bClampLinearTranslationLimitToRefPose=true
    bOverrideWorldGravity=true
    overridePhysicsAsset=&/F_MED_MagneticPlotter_Sleeves_Physics
    simulationSpace=BaseBoneSpace
    alphaBoolBlend {blendOption=Linear}
    alphaScaleBiasClamp {bInterpResult=true}
    baseBoneRef {boneName=pelvis}
   }
   Save cached pose 'sleevesDyn'(Pose=Component To Local(ComponentPose=@n39)) {cacheName=sleevesDyn}
  ]}
 EventGraph {nodes [
   Event BlueprintUpdateAnimation()
   > n1 = Sequence!(execute) -> (then_0, then_1, then_2:Exec, then_3:Exec)
   > n2 = Gravity Override State Function DEPRECATED(
    CharacterPartAnimInstance=Self-Reference()
    GameTimeInSeconds=GetGameTimeInSeconds()
    GravityOverrideParams=Get Gravity Override Params()
    RBAN Params=Get RBAN Params()
   )
   > Set GravityOverride(GravityOverride=@n2)
   > Set Sim Space Settings(Sim Space Settings:Sim Space Settings Structure=Break RBANSetting Struct!(RBANSettingStruct:"RBANSetting Struct Structure (by ref)"=@n2.RBAN SettingsOut) -> (
     ComponentLinearAccScale_28_#1:Vector=~
     ComponentLinearVelScale_31_#2:Vector=~
     SimSpaceSettings_34_#3:Sim Space Settings Structure
    ) {is=title K2Node_BreakStruct}.SimSpaceSettings_34_#3) -> (Output_Get:Sim Space Settings Structure)
   n3 = Get bRigidBodyEnabled()
   > Set Jumping(Jumping:bool=AND Boolean(A=Get bIsJumping(), B=@n3)) -> (Output_Get:bool)
   > Set RBAN OFF(RBAN OFF:bool=NOT Boolean(A=@n3)) -> (Output_Get:bool)
   > n4 = Branch(Condition=Get bIsSprinting())
   > Set rban sleeves(rban sleeves:double=0.35) -> (Output_Get:double)
   Set rban sleeves(execute=@n4.else, rban sleeves=1)
   n5 = Get bIsInFrontEnd()
   n6 = Branch(execute=@n1.then_1, Condition=OR Boolean!(A=Get bIsPlayingEmote(), B=Get bIsSliding(), C:bool=@n5) -> (ReturnValue))
   > Set playingEmote(playingEmote:double=0) -> (Output_Get:double)
   n7 = Branch(execute=@n6.else, Condition=Get bIsSkydiveFloating())
   > Set playingEmote(playingEmote=0.5)
   Set playingEmote(execute=@n7.else, playingEmote=1)
   Set FrontEnd(execute=@n1.then_3, FrontEnd:bool=AND Boolean(A=@n5, B=Get bRigidBodyEnabled())) -> (Output_Get:bool)
   n8 = Event BlueprintProcessAnimInputEvent()
   n9 = IsNamed(self=@n8.AnimInputEvent, InEventName=aie.Player.Zipline.Entered)
   n10 = OR Boolean!(A=Get bIsMoving2D(), B=Get bIsTacticalSprinting()) -> (ReturnValue)
   n11 = Get bIsJumping()
   n12 = Branch(execute=@n1.then_2, Condition=OR Boolean(A=@n9, B=AND Boolean!(A=@n9, B=@n11, C:bool=@n10) -> (ReturnValue)))
   Branch(execute=@n12.else, Condition=IsNamed(self=@n8.AnimInputEvent, InEventName=aie.Player.Zipline.Exited))
   > Set OnZipline(OnZipline:bool=false) -> (Output_Get:bool)
   > Set animDyn on zipline(animDyn on zipline:Vector="0.3,0.3,0.3") -> (Output_Get:Vector)
   Set OnZipline(OnZipline=true)
   > Set animDyn on zipline(animDyn on zipline="0, 0, 0")
  ]}
]
variables [
 animDyn on zipline {type=Vector, default=<0.3 0.3 0.3>}
 FrontEnd {type=boolean}
 Gravity Override Params {
  type=GravityOverrideParamsStruct
  default {
   front_End {gravityOverride=<0 0 -970>}
   skydive_FloatingDown {gravityOverride=<0 -700 500>, windAmplitude=<200 -500 0>}
   skydive_GoingUp {gravityOverride=<0 -700 250>, windAmplitude=<200 -500 0>, windFrequency=<0 0 0>, windOffset=<0 -600 400>}
   skydive_Parachuting {gravityOverride=<0 -450 -800>}
   tactical_Sprint {gravityOverride=<0 -450 -700>}
   ziplining {gravityOverride=<0 0 -100>}
  }
 }
 GravityOverride {type=Vector}
 Jumping {type=boolean}
 OnZipline {type=boolean}
 playingEmote {type=number}
 RBAN OFF {type=boolean}
 RBAN Params {
  type=RBANParamsStruct
  default {
   dBNO {simSpaceSettings {
     maxAngularAcceleration=500
     maxAngularVelocity=3
     maxLinearAcceleration=400
     maxLinearVelocity=250
     velocityScaleZ=0.2
     worldAlpha=0.2
     externalLinearDragV=<1 1 1>
    }}
   dBNO_Carried {simSpaceSettings {
     maxAngularAcceleration=500
     maxAngularVelocity=3
     maxLinearAcceleration=400
     maxLinearVelocity=250
     velocityScaleZ=0.2
     worldAlpha=0.2
     externalLinearDragV=<1 1 1>
    }}
   emoteOrMelee {simSpaceSettings {
     maxAngularAcceleration=500
     maxAngularVelocity=3
     maxLinearAcceleration=400
     maxLinearVelocity=500
     velocityScaleZ=1
     externalLinearDragV=<1 1 1>
    }}
   front_End {simSpaceSettings {
     maxAngularAcceleration=100
     maxAngularVelocity=1
     maxLinearAcceleration=100
     maxLinearVelocity=100
     velocityScaleZ=0
     worldAlpha=0.3
     externalLinearDragV=<0.1 0.1 0.1>
    }}
   onGround_Crouch_Moving {simSpaceSettings {
     maxAngularAcceleration=500
     maxAngularVelocity=3
     maxLinearAcceleration=400
     maxLinearVelocity=250
     velocityScaleZ=0.5
     worldAlpha=0.5
     externalLinearDragV=<1 1 1>
    }}
   onGround_Falling {simSpaceSettings {
     maxAngularAcceleration=500
     maxAngularVelocity=3
     maxLinearAcceleration=400
     maxLinearVelocity=250
     velocityScaleZ=0.5
     worldAlpha=0.5
     externalLinearDragV=<1 1 1>
    }}
   onGround_Moving {
    componentLinearAccScale=<0 0 0>
    componentLinearVelScale=<0 0 0>
    simSpaceSettings {
     maxAngularAcceleration=500
     maxAngularVelocity=3
     maxLinearAcceleration=400
     maxLinearVelocity=250
     velocityScaleZ=0.5
     worldAlpha=0.5
     externalLinearDragV=<1 1 1>
    }
   }
   onGround_Standing {simSpaceSettings {
     maxAngularAcceleration=500
     maxAngularVelocity=3
     maxLinearAcceleration=400
     maxLinearVelocity=250
     velocityScaleZ=0.5
     worldAlpha=0.5
     externalLinearDragV=<1 1 1>
    }}
   rail_Grinding {simSpaceSettings {
     maxAngularAcceleration=500
     maxAngularVelocity=3
     maxLinearAcceleration=400
     maxLinearVelocity=250
     velocityScaleZ=0.5
     worldAlpha=0.5
     externalLinearDragV=<1 1 1>
    }}
   riding_Moving {simSpaceSettings {
     maxAngularAcceleration=500
     maxAngularVelocity=3
     maxLinearAcceleration=400
     maxLinearVelocity=250
     velocityScaleZ=0.5
     worldAlpha=0.5
     externalLinearDragV=<1 1 1>
    }}
   riding_Moving_Targeting {simSpaceSettings {
     maxAngularAcceleration=500
     maxAngularVelocity=3
     maxLinearAcceleration=400
     maxLinearVelocity=250
     velocityScaleZ=0.5
     worldAlpha=0.5
     externalLinearDragV=<1 1 1>
    }}
   riding_Standing {simSpaceSettings {
     maxAngularAcceleration=500
     maxAngularVelocity=3
     maxLinearAcceleration=400
     maxLinearVelocity=250
     velocityScaleZ=0.5
     worldAlpha=0.5
     externalLinearDragV=<1 1 1>
    }}
   skydive_DivingDown {simSpaceSettings {
     maxAngularAcceleration=500
     maxAngularVelocity=3
     maxLinearAcceleration=400
     maxLinearVelocity=250
     velocityScaleZ=0.5
     worldAlpha=0.2
     externalLinearDragV=<1 1 1>
    }}
   skydive_FloatingDown {
    componentLinearAccScale=<0 0 0>
    componentLinearVelScale=<0 0 0>
    simSpaceSettings {
     maxAngularAcceleration=500
     maxAngularVelocity=3
     maxLinearAcceleration=400
     maxLinearVelocity=250
     velocityScaleZ=0.5
     worldAlpha=0.5
     externalLinearDragV=<1 1 1>
    }
   }
   skydive_GoingUp {simSpaceSettings {
     maxAngularAcceleration=500
     maxAngularVelocity=3
     maxLinearAcceleration=400
     maxLinearVelocity=250
     velocityScaleZ=0.5
     worldAlpha=0.5
     externalLinearDragV=<1 1 1>
    }}
   skydive_Parachuting {
    componentLinearAccScale=<0 0.1 0.3>
    componentLinearVelScale=<0 0.1 0.3>
    simSpaceSettings {
     maxAngularAcceleration=500
     maxAngularVelocity=3
     maxLinearAcceleration=400
     maxLinearVelocity=250
     velocityScaleZ=0.2
     worldAlpha=0.2
     externalLinearDragV=<1 1 1>
    }
   }
   skydive_Surfing {simSpaceSettings {
     maxAngularAcceleration=500
     maxAngularVelocity=3
     maxLinearAcceleration=400
     maxLinearVelocity=250
     velocityScaleZ=0.2
     worldAlpha=0.2
     externalLinearDragV=<1 1 1>
    }}
   swimming {simSpaceSettings {
     maxAngularAcceleration=500
     maxAngularVelocity=3
     maxLinearAcceleration=400
     maxLinearVelocity=250
     velocityScaleZ=0.5
     worldAlpha=0.5
     externalLinearDragV=<1 1 1>
    }}
   tactical_Sprint {simSpaceSettings {
     maxAngularAcceleration=500
     maxAngularVelocity=3
     maxLinearAcceleration=400
     maxLinearVelocity=250
     velocityScaleZ=0.5
     worldAlpha=0.5
     externalLinearDragV=<1 1 1>
    }}
   ziplining {simSpaceSettings {
     maxAngularAcceleration=500
     maxAngularVelocity=3
     maxLinearAcceleration=400
     maxLinearVelocity=250
     velocityScaleZ=0.1
     worldAlpha=0.1
     externalLinearDragV=<1 1 1>
    }}
  }
 }
 rban sleeves {type=number}
 Sim Space Settings {
  type=SimSpaceSettings
  default {
   dampingAlpha=1
   maxAngularAcceleration=10000
   maxAngularVelocity=10000
   maxLinearAcceleration=10000
   maxLinearVelocity=10000
   velocityScaleZ=1
   worldAlpha=0
   externalLinearDragV=<0 0 0>
  }
 }
 Tag Container {type=GameplayTagContainer, default {gameplayTags []}}
 Tag Query {
  type=GameplayTagQuery
  default {autoDescription=" ANY( )", tokenStreamVersion=0, userDescription="", queryTokenStream [0,1,1,0], tagDictionary []}
 }
]
