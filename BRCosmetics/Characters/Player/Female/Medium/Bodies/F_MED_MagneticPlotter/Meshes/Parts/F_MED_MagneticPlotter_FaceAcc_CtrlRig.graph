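# Control Rig for the MagneticPlotter female character's face accessories, correcting cap and back hair bones as the head moves.
# Two rivet bones are parent-constrained equally between head and spine_05, and aim constraints point both base_cap and base_main_back_hair at the cap rivet.
# A DeltaRotation function remaps head-to-neck_02 Euler Z into a weight that blends a corrective transform, including squash, onto the back hair base.
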ControlRigBlueprint /BRCosmetics/Characters/Player/Female/Medium/Bodies/F_MED_MagneticPlotter/Meshes/Parts/F_MED_MagneticPlotter_FaceAcc_CtrlRig
graphs [
 New Function_ContainedGraph {nodes [
   RigVMFunctionEntryNode()
   > &RigUnit_ParentConstraint(
    AdvancedSettings=(InterpolationType=Average,RotationOrderForFilter=XZY)
    Child=(Type=Bone,Name="rivet_for_aim_cap")
    Parents=((Item=(Type=Bone,Name="head"),Weight=1),(Item=(Type=Bone,Name="spine_05"),Weight=1))
    Weight=1
   )
   > &RigUnit_AimConstraintLocalSpaceOffset(
    Child=(Type=Bone,Name="base_cap")
    Filter=(bX=True,bY=True,bZ=True)
    Parents=((Item=(Type=Bone,Name="rivet_for_aim_cap"),Weight=1))
    Weight=1
    WorldUp=(Target=(X=0,Y=0,Z=1),Kind=Direction,Space=(Type=Bone,Name="rivet_for_aim_cap"))
   )
   > RigVMFunctionReturnNode()
  ]}
 New Function_ContainedGraph {nodes [
   RigVMFunctionEntryNode()
   > &RigUnit_ParentConstraint(
    AdvancedSettings=(InterpolationType=Average,RotationOrderForFilter=XZY)
    Child=(Type=Bone,Name="rivet_for_aim_hair")
    Parents=((Item=(Type=Bone,Name="head"),Weight=1),(Item=(Type=Bone,Name="spine_05"),Weight=1))
    Weight=1
   )
   > &RigUnit_AimConstraintLocalSpaceOffset(
    Child=(Type=Bone,Name="base_main_back_hair")
    Filter=(bX=True,bY=True,bZ=True)
    Parents=((Item=(Type=Bone,Name="rivet_for_aim_cap"),Weight=1))
    Weight=1
    WorldUp=(Target=(X=0,Y=0,Z=1),Kind=Direction,Space=(Type=Bone,Name="rivet_for_aim_cap"))
   )
   > DeltaRotation = RigVMFunctionReferenceNode(Child=head, Parent=neck_02, Source Max=-60, Source Min=-25)
   > &RigUnit_ModifyTransforms(
    ItemToModify=((Item=(Type=Bone,Name="base_main_back_hair"),Transform=(Rotation=(X=0.0016,Y=0.0001,Z=0.0377,W=0.9993),Translation=(X=4.8,Y=-0.4415,Z=0.4874),Scale3D=(X=1,Y=0.7825,Z=0.6883))))
    WeightMinimum=0
    Weight=@DeltaRotation.Delta_z
   )
   > RigVMFunctionReturnNode()
   &RigUnit_ModifyTransforms(
    ExecutePin=()
    ItemToModify=((Item=(Type=Bone,Name="base_main_back_hair"),Transform=(Rotation=(X=0.0016,Y=0.0001,Z=0.0377,W=0.9993),Translation=(X=4.3861,Y=-0.4415,Z=0.4874),Scale3D=(X=1,Y=0.7825,Z=0.6883))))
    Weight=1
    WeightMinimum=0
   )
  ]}
 New Function_ContainedGraph {nodes [
   n1 = RigVMFunctionEntryNode()
   RigVMFunction_MathQuaternionToEuler = &RigVMFunction_MathQuaternionToEuler(
    RotationOrder=ZYX
    Value=&RigUnit_GetRelativeTransformForItem(
     Child=(Type=Bone,Name="head")
     Parent=(Type=Bone,Name="neck_01")
     bChildInitial=False
     bParentInitial=False
     Child.Name=@n1.Child
     Parent.Name=@n1.Parent
    ).RelativeTransform.Rotation
   )
   Remap = &RigVMFunction_MathDoubleRemap(
    TargetMaximum=1
    TargetMinimum=0
    SourceMaximum=@n1.Source Max
    SourceMinimum=@n1.Source Min
    Value=@RigVMFunction_MathQuaternionToEuler.Result.Z
   )
   > RigVMFunctionReturnNode(Delta_z=@Remap.Result)
  ]}
 RigVMFunctionLibrary {nodes [
   DeltaRotation = RigVMCollapseNode(ExecuteContext=())
   cap = RigVMCollapseNode(ExecuteContext=())
   hair = RigVMCollapseNode(ExecuteContext=())
  ]}
 RigVMModel {nodes [
   RigUnit_BeginExecution = &RigUnit_BeginExecution()
   cap = RigVMFunctionReferenceNode(ExecuteContext=@RigUnit_BeginExecution.ExecutePin)
   > hair = RigVMFunctionReferenceNode()
  ]}
]
hierarchy [
 root {loc=<0 0 0>, rot=<0 0 0 1>}
 pelvis {parent=root, loc=<0 -0.5301 98.0571>, rot=<0.0127 0.707 -0.0127 -0.707>}
 spine_01 {parent=pelvis, loc=<6.1208 -0.9328 0>, rot=<0 0 0.034 -0.9994>}
 spine_02 {parent=spine_01, loc=<6.5997 -0.2095 0>, rot=<0 0 0.0238 -0.9997>}
 spine_03 {parent=spine_02, loc=<6.4914 -0.1237 0>, rot=<0 0 0.0149 -0.9999>}
 spine_04 {parent=spine_03, loc=<6.9598 0.0794 0>, rot=<0 0 -0.0132 -0.9999>}
 spine_05 {parent=spine_04, loc=<6.9987 0 0>, rot=<0 0 -0.0015 -1>}
 neck_01 {parent=spine_05, loc=<11.3138 0.1529 0>, rot=<0 0 -0.2003 -0.9797>}
 neck_02 {parent=neck_01, loc=<4.5627 0 0>, rot=<0 0 0.0504 -0.9987>}
 head {parent=neck_02, loc=<5.3914 0 0>, rot=<0 0 0.115 -0.9934>}
 base_main_front_hair {parent=head, loc=<8.4691 10.1649 0.1758>, rot=<0.518 0.4813 0.522 0.477>}
 dyn_main_front_hair {parent=base_main_front_hair, loc=<0 0 0>, rot=<0 0 0 1>}
 dyn_front_hair_0 {parent=dyn_main_front_hair, loc=<0.0405 -1.1288 -1.2071>, rot=<-0.0791 -0.4201 -0.0944 -0.8991>}
 dyn_front_hair_1 {parent=dyn_main_front_hair, loc=<2.3764 1.1468 -1.3802>, rot=<-0.1703 -0.4175 -0.1845 -0.8733>}
 dyn_front_hair_2 {parent=dyn_main_front_hair, loc=<2.7765 2.6626 0.5954>, rot=<0.0003 -0.3962 -0.4111 -0.821>}
 base_main_back_hair {parent=head, loc=<1.8426 -9.6915 0.0198>, rot=<0.5021 -0.4979 0.4979 -0.5021>}
 dyn_main_back_hair {parent=base_main_back_hair, loc=<0 0 0>, rot=<0 0 0 1>}
 base_back_hair_c_0 {parent=dyn_main_back_hair, loc=<3.7067 0 -0.5118>, rot=<0 -0.3007 0 -0.9537>}
 dyn_back_hair_c_0 {parent=base_back_hair_c_0, loc=<0 0 0>, rot=<0 0 0 1>}
 dyn_back_hair_c_1 {parent=dyn_back_hair_c_0, loc=<4 0 0>, rot=<0 0 0 1>}
 base_back_hair_l_0 {parent=dyn_main_back_hair, loc=<3.6956 4.9918 -0.5118>, rot=<-0.0028 -0.3109 -0.2206 -0.9245>}
 dyn_back_hair_l_0 {parent=base_back_hair_l_0, loc=<0 0 0>, rot=<0 0 0 1>}
 dyn_back_hair_l_1 {parent=dyn_back_hair_l_0, loc=<4 0 0>, rot=<0 0 0 1>}
 base_back_hair_r_0 {parent=dyn_main_back_hair, loc=<3.6956 -5.017 -0.5118>, rot=<0.0028 -0.3109 0.2206 -0.9245>}
 dyn_back_hair_r_0 {parent=base_back_hair_r_0, loc=<0 0 0>, rot=<0 0 0 1>}
 dyn_back_hair_r_1 {parent=dyn_back_hair_r_0, loc=<4 0 0>, rot=<0 0 0 1>}
 base_main_back_hair_new {parent=head, loc=<1.8426 -9.6915 0.0198>, rot=<0.0806 -0.7025 0.7025 -0.0806>}
 dyn_main_back_hair_new {parent=base_main_back_hair_new, loc=<0 0 0>, rot=<0 0 0 1>}
 cap1 {parent=head, loc=<5.9913 -9.1509 0>, rot=<0.5816 -0.4021 0.4021 -0.5816>}
 cap {parent=cap1, loc=<0 0 0>, rot=<0 0 0 1>}
 base_cap {parent=head, loc=<5.9913 -9.1509 0>, rot=<0.5816 -0.4021 0.4021 -0.5816>}
 dyn_cap {parent=base_cap, loc=<0 0 0>, rot=<0 0 0 1>}
 rivet_for_aim_cap {parent=head, loc=<13.7863 -29.8033 0>, rot=<0.5816 -0.4021 0.4021 -0.5816>}
 rivet_for_aim_hair {parent=head, loc=<2.0079 -29.6908 0.0198>, rot=<0.5021 -0.4979 0.4979 -0.5021>}
 clavicle_l {parent=spine_05, loc=<5.4831 -0.1053 -3.2946>, rot=<0.0465 -0.7289 -0.009 -0.6829>}
 upperarm_l {parent=clavicle_l, loc=<9.8658 -1.2136 -0.4105>, rot=<0.0164 -0.3245 -0.0269 -0.9454>}
 clavicle_r {parent=spine_05, loc=<5.4831 -0.1053 3.2946>, rot=<0.7289 0.0465 -0.6829 0.009>}
 upperarm_r {parent=clavicle_r, loc=<-9.8658 1.2136 0.4105>, rot=<0.0164 -0.3245 -0.0269 -0.9454>}
]
