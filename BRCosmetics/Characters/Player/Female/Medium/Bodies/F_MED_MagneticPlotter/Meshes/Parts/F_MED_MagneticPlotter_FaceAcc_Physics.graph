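# Physics asset for the F_MED_MagneticPlotter face accessory, covering the back hair of a female player cosmetic.
# Capsules form center, left and right two-segment dynamic chains on small base bodies, alongside larger spine_05 and clavicle bodies.
# Chain joints have narrow swing limits and zero twist limit.
# Angularly free constraints tie the bases to spine_05 and cross-link the first center segment to its left and right neighbors.
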PhysicsAsset /BRCosmetics/Characters/Player/Female/Medium/Bodies/F_MED_MagneticPlotter/Meshes/Parts/F_MED_MagneticPlotter_FaceAcc_Physics
bodies {
 base_back_hair_c_0 [{shapeType=Capsule, length=0.505, radius=0.505, shapeName=base_back_hair_c_0_capsule, rotation=<rot 0 -90 0>}]
 base_back_hair_l_0 [{shapeType=Capsule, length=0.505, radius=0.505, shapeName=base_back_hair_l_0_capsule, rotation=<rot 0 -90 0>}]
 base_back_hair_r_0 [{shapeType=Capsule, length=0.505, radius=0.505, shapeName=base_back_hair_r_0_capsule, rotation=<rot 0 -90 0>}]
 base_main_back_hair [{shapeType=Capsule, length=0.505, radius=0.505, shapeName=base_main_back_hair_capsule, rotation=<rot 0 -90 0>}]
 clavicle_l [{shapeType=Capsule, length=15.0448, radius=7.6011, shapeName=clavicle_l_capsule, center=<5 -1.32 0.7>, rotation=<rot 90 0 0>}]
 clavicle_r [{shapeType=Capsule, length=15.0448, radius=7.6011, shapeName=clavicle_r_capsule, center=<-5 1.32 -0.7>, rotation=<rot 90 0 0>}]
 dyn_back_hair_c_0 [{shapeType=Capsule, length=5.349, radius=2.484, shapeName=dyn_back_hair_c_0_capsule, center=<4 0 -3.5>, rotation=<rot 0 0 90>}]
 dyn_back_hair_c_1 [{shapeType=Capsule, length=4.5333, radius=3.2996, shapeName=dyn_back_hair_c_1_capsule, center=<4 0 -4>, rotation=<rot 0 0 90>}]
 dyn_back_hair_l_0 [{shapeType=Capsule, length=5.349, radius=2.4839, shapeName=dyn_back_hair_l_0_capsule, center=<4 0 -3.5>, rotation=<rot 0 0 90>}]
 dyn_back_hair_l_1 [{shapeType=Capsule, length=5.1456, radius=2.6874, shapeName=dyn_back_hair_l_1_capsule, center=<2.3715 0 -5.673>, rotation=<rot 0 0 90>}]
 dyn_back_hair_r_0 [{shapeType=Capsule, length=4.669, radius=3.164, shapeName=dyn_back_hair_r_0_capsule, center=<4 0 -4>, rotation=<rot -20.6994 -0.0022 90.0033>}]
 dyn_back_hair_r_1 [{shapeType=Capsule, length=4.419, radius=3.4139, shapeName=dyn_back_hair_r_1_capsule, center=<4 0 -4>, rotation=<rot -14.4704 0 90>}]
 spine_05 [{shapeType=Capsule, length=12.4717, radius=11.9718, shapeName=spine_05_capsule, center=<0.7257 0.6547 0>, rotation=<rot 0 90 0>}]
}
constraints [
 {bone1Name=dyn_back_hair_c_0, bone2Name=base_back_hair_c_0, swing1LimitDegrees=10, swing2LimitDegrees=10, twistLimitDegrees=0}
 {bone1Name=dyn_back_hair_c_1, bone2Name=dyn_back_hair_c_0, swing1LimitDegrees=20, swing2LimitDegrees=10, twistLimitDegrees=0}
 {bone1Name=dyn_back_hair_l_0, bone2Name=base_back_hair_l_0, swing1LimitDegrees=20, swing2LimitDegrees=20, twistLimitDegrees=0}
 {bone1Name=dyn_back_hair_l_1, bone2Name=dyn_back_hair_l_0, swing1LimitDegrees=20, swing2LimitDegrees=10, twistLimitDegrees=0}
 {bone1Name=dyn_back_hair_r_0, bone2Name=base_back_hair_r_0, swing1LimitDegrees=20, swing2LimitDegrees=10, twistLimitDegrees=0}
 {bone1Name=dyn_back_hair_r_1, bone2Name=dyn_back_hair_r_0, swing1LimitDegrees=20, swing2LimitDegrees=10, twistLimitDegrees=0}
 {
  bone1Name=base_main_back_hair
  bone2Name=spine_05
  swing1LimitDegrees=45
  swing1Motion=Free
  swing2LimitDegrees=45
  swing2Motion=Free
  twistLimitDegrees=45
  twistMotion=Free
 }
 {
  bone1Name=base_back_hair_c_0
  bone2Name=spine_05
  swing1LimitDegrees=45
  swing1Motion=Free
  swing2LimitDegrees=45
  swing2Motion=Free
  twistLimitDegrees=45
  twistMotion=Free
 }
 {
  bone1Name=base_back_hair_l_0
  bone2Name=spine_05
  swing1LimitDegrees=45
  swing1Motion=Free
  swing2LimitDegrees=45
  swing2Motion=Free
  twistLimitDegrees=45
  twistMotion=Free
 }
 {
  bone1Name=base_back_hair_r_0
  bone2Name=spine_05
  swing1LimitDegrees=45
  swing1Motion=Free
  swing2LimitDegrees=45
  swing2Motion=Free
  twistLimitDegrees=45
  twistMotion=Free
 }
 {
  bone1Name=dyn_back_hair_c_0
  bone2Name=dyn_back_hair_l_0
  swing1LimitDegrees=45
  swing1Motion=Free
  swing2LimitDegrees=45
  swing2Motion=Free
  twistLimitDegrees=45
  twistMotion=Free
 }
 {
  bone1Name=dyn_back_hair_r_0
  bone2Name=dyn_back_hair_c_0
  swing1LimitDegrees=45
  swing1Motion=Free
  swing2LimitDegrees=45
  swing2Motion=Free
  twistLimitDegrees=45
  twistMotion=Free
 }
]
