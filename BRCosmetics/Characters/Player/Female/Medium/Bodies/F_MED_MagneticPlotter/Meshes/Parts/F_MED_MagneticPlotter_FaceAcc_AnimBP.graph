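# Post-process style Anim Blueprint for the MagneticPlotter female face accessory part, adding secondary motion to front hair and physics bones.
# It copies the parent mesh pose, then runs a Control Rig, four chained spring AnimDynamics hair bones, and a RigidBody solve based at spine_05.
# The event graph updates gravity and sim-space settings from movement-state parameters, picks simulation space by front-end state, and toggles a dynamics vector on zipline events.
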AnimBlueprint /BRCosmetics/Characters/Player/Female/Medium/Bodies/F_MED_MagneticPlotter/Meshes/Parts/F_MED_MagneticPlotter_FaceAcc_AnimBP
target_skeleton=&/F_MED_MagneticPlotter_FaceAcc_Skeleton
graphs [
 AnimGraph {nodes [
   Output Pose(Result=Use cached pose 'RigidCache'()) {layerGroup=None}
   Save cached pose 'inputCache'(Pose=Copy Pose From Mesh() {bUseAttachedParent=true}) {cacheName=inputCache}
   n1 = Control Rig(Source=Use cached pose 'inputCache'()) {
    alphaBoolBlend {blendOption=Linear}
    controlRigAssetReference {blueprintRigClass=&/F_MED_MagneticPlotter_FaceAcc_CtrlRig.F_MED_MagneticPlotter_FaceAcc_CtrlRig_C}
   }
   Save cached pose 'controlRigCache'(Pose=@n1) {cacheName=controlRigCache}
   n2 = RigidBody!(
    ComponentPose=Local To Component(LocalPose=Use cached pose 'dynCache'())
    Alpha
    OverrideWorldGravity:Vector="0,0,0"
    ExternalForce
    SimSpaceSettings:Sim Space Settings Structure=(WorldAlpha=0.8,VelocityScaleZ=0.7,DampingAlpha=1,MaxLinearVelocity=500,MaxAngularVelocity=3,MaxLinearAcceleration=1000,MaxAngularAcceleration=1,ExternalLinearDragV=(X=1,Y=1,Z=1),ExternalLinearVelocity=(X=0,Y=0,Z=0),ExternalAngularVelocity=(X=0,Y=0,Z=0))
   ) -> (Pose) {
    bClampLinearTranslationLimitToRefPose=true
    bOverrideWorldGravity=true
    overridePhysicsAsset=&/F_MED_MagneticPlotter_FaceAcc_Physics
    simulationSpace=BaseBoneSpace
    alphaBoolBlend {blendOption=Linear}
    baseBoneRef {boneName=spine_05}
    simSpaceSettings {
     maxAngularAcceleration=1
     maxAngularVelocity=3
     maxLinearAcceleration=1000
     maxLinearVelocity=500
     velocityScaleZ=0.7
     worldAlpha=0.8
     externalLinearDragV=<1 1 1>
    }
   }
   Save cached pose 'RigidCache'(Pose=Component To Local(ComponentPose=@n2)) {cacheName=RigidCache}
   n3 = Anim Dynamics - Bone: dyn_main_front_hair!(
    SimulationSpace:AnimPhysSimSpaceType Enum
    ComponentPose=Local To Component(LocalPose=Use cached pose 'controlRigCache'())
    Alpha=1
    ComponentLinearAccScale:Vector="1,1,1"
    ComponentLinearVelScale:Vector="1,1,1"
   ) -> (Pose) {
    angularDampingOverride=1
    angularSpringConstant=20000
    bAngularSpring=true
    bOverrideAngularDamping=true
    bUseGravityOverride=true
    gravityScale=0.35
    simulationSpace=RootRelative
    alphaBoolBlend {blendOption=Linear}
    boundBone {boneName=dyn_main_front_hair}
    componentAppliedLinearAccClamp=<5000 5000 5000>
    componentLinearAccScale=<1 1 1>
    componentLinearVelScale=<1 1 1>
    gravityOverride {z=-980}
    physicsBodyDefinitions [{
      boundBone {boneName=dyn_main_front_hair}
      boxExtents=<20 50 50>
      constraintSetup {angularLimitsMax=<0 5 30>, angularLimitsMin=<0 -70 -30>, angularTarget=<1 0 0>}
      localJointOffset=<8 0 0>
     }]
    relativeSpaceBone {boneName=root}
   }
   n4 = Anim Dynamics - Bone: dyn_front_hair_0!(
    SimulationSpace:AnimPhysSimSpaceType Enum
    ComponentPose=@n3
    Alpha=1
    ComponentLinearAccScale:Vector="1,1,1"
    ComponentLinearVelScale:Vector="1,1,1"
   ) -> (Pose) {
    angularDampingOverride=1
    angularSpringConstant=20000
    bAngularSpring=true
    bOverrideAngularDamping=true
    bUseGravityOverride=true
    gravityScale=0.35
    simulationSpace=RootRelative
    alphaBoolBlend {blendOption=Linear}
    boundBone {boneName=dyn_front_hair_0}
    componentAppliedLinearAccClamp=<5000 5000 5000>
    componentLinearAccScale=<1 1 1>
    componentLinearVelScale=<1 1 1>
    gravityOverride {z=-980}
    physicsBodyDefinitions [{
      boundBone {boneName=dyn_front_hair_0}
      boxExtents=<30 50 50>
      constraintSetup {angularLimitsMax=<0 12 9>, angularLimitsMin=<0 -50 -50>, angularTarget=<1 0 0>}
      localJointOffset=<8 0 0>
     }]
    relativeSpaceBone {boneName=root}
   }
   n5 = Anim Dynamics - Bone: dyn_front_hair_1!(
    SimulationSpace:AnimPhysSimSpaceType Enum
    ComponentPose=@n4
    Alpha=1
    ComponentLinearAccScale:Vector="1,1,1"
    ComponentLinearVelScale:Vector="1,1,1"
   ) -> (Pose) {
    angularDampingOverride=1
    angularSpringConstant=20000
    bAngularSpring=true
    bOverrideAngularDamping=true
    bUseGravityOverride=true
    gravityScale=0.35
    simulationSpace=RootRelative
    alphaBoolBlend {blendOption=Linear}
    boundBone {boneName=dyn_front_hair_1}
    componentAppliedLinearAccClamp=<5000 5000 5000>
    componentLinearAccScale=<1 1 1>
    componentLinearVelScale=<1 1 1>
    gravityOverride {z=-980}
    physicsBodyDefinitions [{
      boundBone {boneName=dyn_front_hair_1}
      boxExtents=<30 50 50>
      constraintSetup {angularLimitsMax=<0 80 9>, angularLimitsMin=<0 -50 -50>, angularTarget=<1 0 0>}
      localJointOffset=<8 0 0>
     }]
    relativeSpaceBone {boneName=root}
   }
   n6 = Anim Dynamics - Bone: dyn_front_hair_2!(
    SimulationSpace:AnimPhysSimSpaceType Enum
    ComponentPose=@n5
    Alpha=1
    ComponentLinearAccScale:Vector="1,1,1"
    ComponentLinearVelScale:Vector="1,1,1"
   ) -> (Pose) {
    angularDampingOverride=1
    angularSpringConstant=20000
    bAngularSpring=true
    bOverrideAngularDamping=true
    bUseGravityOverride=true
    gravityScale=0.35
    simulationSpace=RootRelative
    alphaBoolBlend {blendOption=Linear}
    boundBone {boneName=dyn_front_hair_2}
    componentAppliedLinearAccClamp=<5000 5000 5000>
    componentLinearAccScale=<1 1 1>
    componentLinearVelScale=<1 1 1>
    gravityOverride {z=-980}
    physicsBodyDefinitions [{
      boundBone {boneName=dyn_front_hair_2}
      boxExtents=<30 50 50>
      constraintSetup {angularLimitsMax=<0 80 9>, angularLimitsMin=<0 -50 -50>, angularTarget=<1 0 0>}
      localJointOffset=<8 0 0>
     }]
    relativeSpaceBone {boneName=root}
   }
   Save cached pose 'dynCache'(Pose=Component To Local(ComponentPose=@n6)) {cacheName=dynCache}
  ]}
 EventGraph {nodes [
   Event BlueprintUpdateAnimation()
   > Sequence()
   > n1 = Gravity Override State Function DEPRECATED(
    CharacterPartAnimInstance=Self-Reference()
    GameTimeInSeconds=GetGameTimeInSeconds()
    GravityOverrideParams=Get Gravity Override Params()
    RBAN Params=Get RBAN Params()
   )
   > Set GravityOverride(GravityOverride=@n1)
   > Set Sim Space Settings(Sim Space Settings:Sim Space Settings Structure=Break RBANSetting Struct!(RBANSettingStruct:"RBANSetting Struct Structure (by ref)"=@n1.RBAN SettingsOut) -> (
     ComponentLinearAccScale_28_#1:Vector=~
     ComponentLinearVelScale_31_#2:Vector=~
     SimSpaceSettings_34_#3:Sim Space Settings Structure
    ) {is=title K2Node_BreakStruct}.SimSpaceSettings_34_#3) -> (Output_Get:Sim Space Settings Structure)
   > n2 = Branch(Condition=Get bIsInFrontEnd())
   > Set Simulation Space(Simulation Space:Byte=2) -> (Output_Get:Byte=0)
   Set Simulation Space(execute=@n2.else, Simulation Space=3) -> (Output_Get=0)
   n3 = Event BlueprintProcessAnimInputEvent()
   n4 = IsNamed(self=@n3.AnimInputEvent, InEventName=aie.Player.Zipline.Entered)
   n5 = OR Boolean(A=Get bIsMoving2D(), B=Get bIsTacticalSprinting())
   n6 = Get bIsJumping()
   > n7 = Branch(Condition=OR Boolean(A=@n4, B=AND Boolean!(A=@n4, B=@n6, C:bool=@n5) -> (ReturnValue)))
   > Set animDyn on zipline(animDyn on zipline:Vector="0, 0, 0") -> (Output_Get:Vector)
   Branch(execute=@n7.else, Condition=IsNamed(self=@n3.AnimInputEvent, InEventName=aie.Player.Zipline.Exited))
   > Set animDyn on zipline(animDyn on zipline="1,1,1")
  ]}
]
variables [
 animDyn on zipline {type=Vector, default=<1 1 1>}
 Gravity Override Params {type=GravityOverrideParamsStruct}
 GravityOverride {type=Vector}
 RBAN Params {
  type=RBANParamsStruct
  default {
   dBNO {simSpaceSettings {
     maxAngularAcceleration=1
     maxAngularVelocity=3
     maxLinearAcceleration=1000
     maxLinearVelocity=500
     velocityScaleZ=0.7
     worldAlpha=0.8
     externalLinearDragV=<1 1 1>
    }}
   dBNO_Carried {simSpaceSettings {
     maxAngularAcceleration=1
     maxAngularVelocity=3
     maxLinearAcceleration=1000
     maxLinearVelocity=500
     velocityScaleZ=0.7
     worldAlpha=0.8
     externalLinearDragV=<1 1 1>
    }}
   emoteOrMelee {simSpaceSettings {
     maxAngularAcceleration=1
     maxAngularVelocity=3
     maxLinearAcceleration=1000
     maxLinearVelocity=500
     velocityScaleZ=0.7
     worldAlpha=0.8
     externalLinearDragV=<1 1 1>
    }}
   front_End {simSpaceSettings {
     maxAngularAcceleration=1
     maxAngularVelocity=3
     maxLinearAcceleration=1000
     maxLinearVelocity=500
     velocityScaleZ=0.7
     worldAlpha=0.8
     externalLinearDragV=<1 1 1>
    }}
   onGround_Crouch_Moving {simSpaceSettings {
     maxAngularAcceleration=1
     maxAngularVelocity=3
     maxLinearAcceleration=1000
     maxLinearVelocity=500
     velocityScaleZ=0.7
     worldAlpha=0.8
     externalLinearDragV=<1 1 1>
    }}
   onGround_Falling {simSpaceSettings {
     maxAngularAcceleration=1
     maxAngularVelocity=3
     maxLinearAcceleration=1000
     maxLinearVelocity=500
     velocityScaleZ=0.7
     worldAlpha=0.8
     externalLinearDragV=<1 1 1>
    }}
   onGround_Moving {simSpaceSettings {
     maxAngularAcceleration=1
     maxAngularVelocity=3
     maxLinearAcceleration=1000
     maxLinearVelocity=500
     velocityScaleZ=0.7
     worldAlpha=0.8
     externalLinearDragV=<1 1 1>
    }}
   onGround_Standing {simSpaceSettings {
     maxAngularAcceleration=1
     maxAngularVelocity=3
     maxLinearAcceleration=1000
     maxLinearVelocity=500
     velocityScaleZ=0.7
     worldAlpha=0.8
     externalLinearDragV=<1 1 1>
    }}
   rail_Grinding {simSpaceSettings {
     maxAngularAcceleration=1
     maxAngularVelocity=3
     maxLinearAcceleration=1000
     maxLinearVelocity=500
     velocityScaleZ=0.7
     worldAlpha=0.8
     externalLinearDragV=<1 1 1>
    }}
   riding_Moving {simSpaceSettings {
     maxAngularAcceleration=1
     maxAngularVelocity=3
     maxLinearAcceleration=1000
     maxLinearVelocity=500
     velocityScaleZ=0.7
     worldAlpha=0.8
     externalLinearDragV=<1 1 1>
    }}
   riding_Moving_Targeting {simSpaceSettings {
     maxAngularAcceleration=1
     maxAngularVelocity=3
     maxLinearAcceleration=1000
     maxLinearVelocity=500
     velocityScaleZ=0.7
     worldAlpha=0.8
     externalLinearDragV=<1 1 1>
    }}
   riding_Standing {simSpaceSettings {
     maxAngularAcceleration=1
     maxAngularVelocity=3
     maxLinearAcceleration=1000
     maxLinearVelocity=500
     velocityScaleZ=0.7
     worldAlpha=0.8
     externalLinearDragV=<1 1 1>
    }}
   skydive_DivingDown {simSpaceSettings {
     maxAngularAcceleration=1
     maxAngularVelocity=3
     maxLinearAcceleration=1000
     maxLinearVelocity=500
     velocityScaleZ=0.7
     worldAlpha=0.8
     externalLinearDragV=<1 1 1>
    }}
   skydive_FloatingDown {
    componentLinearAccScale=<0 0.5 0>
    componentLinearVelScale=<0 0.5 0>
    simSpaceSettings {
     maxAngularAcceleration=1
     maxAngularVelocity=3
     maxLinearAcceleration=1000
     maxLinearVelocity=500
     velocityScaleZ=0.7
     worldAlpha=0.8
     externalLinearDragV=<1 1 1>
    }
   }
   skydive_GoingUp {simSpaceSettings {
     maxAngularAcceleration=1
     maxAngularVelocity=3
     maxLinearAcceleration=1000
     maxLinearVelocity=500
     velocityScaleZ=0.7
     worldAlpha=0.8
     externalLinearDragV=<1 1 1>
    }}
   skydive_Parachuting {
    componentLinearAccScale=<0 0.1 0.3>
    componentLinearVelScale=<0 0.1 0.3>
    simSpaceSettings {
     maxAngularAcceleration=1
     maxAngularVelocity=3
     maxLinearAcceleration=1000
     maxLinearVelocity=500
     velocityScaleZ=0.7
     worldAlpha=0.8
     externalLinearDragV=<1 1 1>
    }
   }
   skydive_Surfing {simSpaceSettings {
     maxAngularAcceleration=1
     maxAngularVelocity=3
     maxLinearAcceleration=1000
     maxLinearVelocity=500
     velocityScaleZ=0.7
     worldAlpha=0.8
     externalLinearDragV=<1 1 1>
    }}
   swimming {simSpaceSettings {
     maxAngularAcceleration=1
     maxAngularVelocity=3
     maxLinearAcceleration=1000
     maxLinearVelocity=500
     velocityScaleZ=0.7
     worldAlpha=0.8
     externalLinearDragV=<1 1 1>
    }}
   tactical_Sprint {simSpaceSettings {
     maxAngularAcceleration=1
     maxAngularVelocity=3
     maxLinearAcceleration=1000
     maxLinearVelocity=500
     velocityScaleZ=0.7
     worldAlpha=0.8
     externalLinearDragV=<1 1 1>
    }}
   ziplining {simSpaceSettings {
     maxAngularAcceleration=1
     maxAngularVelocity=3
     maxLinearAcceleration=1000
     maxLinearVelocity=500
     velocityScaleZ=0.7
     worldAlpha=0.8
     externalLinearDragV=<1 1 1>
    }}
  }
 }
 Sim Space Settings {
  type=SimSpaceSettings
  default {
   dampingAlpha=1
   maxAngularAcceleration=10000
   maxAngularVelocity=10000
   maxLinearAcceleration=10000
   maxLinearVelocity=10000
   velocityScaleZ=1
   worldAlpha=0
   externalLinearDragV=<0 0 0>
  }
 }
 Simulation Space {type=integer}
]
